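# Control Rig for the F_MED_FishLine female cosmetic body that applies pose-driven corrective offsets, mainly to skirt bones.
# It runs the shared DeformRig function, then a sequence of skirt, foot and shoulder correction passes.
# Spherical pose readers on thighs and head, curve-shaped or combined with knee bend, spine rotation and pelvis height, weight additive skirt transforms.
# Foot yaw drives cor_foot bones, and both clavicles' combined rotation drives shoulder and back-skirt offsets.
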ControlRigBlueprint /BRCosmetics/Characters/Player/Female/Medium/Bodies/F_MED_FishLine/Meshes/F_MED_FishLine_CtrlRig
graphs [
 RigVMFunctionLibrary {links [], nodes []}
 RigVMFunction_Sequence_1_ContainedGraph {nodes [
   n1 = RigVMFunctionEntryNode()
   RigVMFunction_Sequence = &RigVMFunction_Sequence(ExecuteContext=@n1.ExecuteContext)
   n2 = &RigVMFunction_Sequence(ExecuteContext=@RigVMFunction_Sequence.B)
   n3 = &RigVMFunction_Sequence(ExecuteContext=@n2.B)
   n4 = &RigVMFunction_Sequence(ExecuteContext=@n3.B)
   n5 = &RigVMFunction_Sequence(ExecuteContext=@n4.B)
   n6 = &RigVMFunction_Sequence(ExecuteContext=@n5.B)
   n7 = &RigVMFunction_Sequence(ExecuteContext=@n6.B)
   RigVMFunctionReturnNode(A=@RigVMFunction_Sequence.A, B=@n2.A, C=@n3.A, D=@n4.A, E=@n5.A, F=@n6.A, G=@n7.A, H=@n7.B)
  ]}
 RigVMModel {nodes [
   RigUnit_BeginExecution = &RigUnit_BeginExecution()
   FN_DeformRig_FNC = RigVMFunctionReferenceNode(
    DeformRig_ Curve="(deformRig_l_upperarm_up_alpha_72_#1=False,deformRig_r_upperarm_up_alpha_73_#2=False,deformRig_l_upperarm_fr_alpha_74_#3=False,deformRig_r_upperarm_fr_alpha_75_CB51B8 ...(truncated)"
    DeformRig_Bool="(upperarm_1_#4=True,elbow_3_#5=True,wrist_7_#6=True,groin_8_#7=True,glute_17_#8=True,knee_9_40C408324CA261 ...(truncated)"
    DeformRig_Debug=(upperarm_debug_18_#4=False,wrist_debug_20_#6=False,lat_debug_26_#9=False,groinGlute_debug_27_#7=False)
    DeformRig_Value="(upperarm_up_18_#10=(Rotation=(X=0,Y=0,Z=0,W=1),Translation=(X=0,Y=0,Z=8),Scale3D=(X=0.7,Y=1,Z=1)),upperarm_fwd_fr_19_#11=(Rotation=(X=0,Y=0,Z=0,W=1),Translation=(X=0,Y=10,Z=0),Scale3D=(X=0.5,Y=1,Z= ...(truncated)"
    ExecuteContext=@RigUnit_BeginExecution.ExecutePin
   )
   > RigVMFunction_Sequence = RigVMAggregateNode()
   n1 = &RigUnit_SphericalPoseReader(
    ActiveRegionScaleFactors=(PositiveWidth=1,NegativeWidth=0.05,PositiveHeight=1,NegativeHeight=1)
    ActiveRegionSize=0.5
    Debug=(bDrawDebug=true,bDraw2D=False,bDrawLocalAxes=False,DebugScale=25,DebugSegments=20,DebugThickness=0.25)
    DriverAxis=(X=1,Y=0,Z=0)
    DriverItem=(Type=Bone,Name="thigh_r")
    FalloffRegionScaleFactors=(PositiveWidth=1,NegativeWidth=1.1,PositiveHeight=1,NegativeHeight=1)
    FalloffSize=0.2
    FlipHeightScaling=False
    FlipWidthScaling=False
    OptionalParentItem=(Type=Bone,Name="pelvis")
    RotationOffset=(X=90,Y=0,Z=178)
    ExecutePin=@RigVMFunction_Sequence.A
   )
   n2 = &RigVMFunction_AnimEvalRichCurve(
    Curve="(EditorCurveData=(Keys=((InterpMode=RCIM_Cubic,ArriveTangent=25.0927),(InterpMode=RCIM_Cubic,Time=0.0508,Value=0.1968,ArriveTangent=1.092,LeaveTangent=1.092),(InterpMode=RCIM_Cubic,TangentMode=RCTM_Break,Time=0.2657,Value=0.2902,ArriveTangent=0.4434,LeaveT ...(truncated)"
    SourceMinimum=0
    TargetMinimum=0
    Value=@n1.OutputParam
   )
   > &RigUnit_ModifyTransforms(
    ItemToModify=((Item=(Type=Bone,Name="MSTR_skirt_f_r"),Transform=(Rotation=(X=0.6344,Y=0.0187,Z=0.2915,W=-0.7157),Translation=(X=-4.3563,Y=-0.2962,Z=4.5843),Scale3D=(X=1,Y=1,Z=1))))
    WeightMinimum=0
    Weight=@n2.Result
   )
   > &RigUnit_ModifyTransforms(
    ItemToModify=((Item=(Type=Bone,Name="MSTR_skirt_d_r"),Transform=(Rotation=(X=-0.1635,Y=-0.0845,Z=-0.2228,W=0.9573),Translation=(X=0,Y=0,Z=0),Scale3D=(X=1,Y=1,Z=1))))
    WeightMinimum=0
    Weight=@n2.Result
   )
   > &RigUnit_ModifyTransforms(
    ItemToModify=((Item=(Type=Bone,Name="MSTR_skirt1_b_r"),Transform=(Rotation=(X=-0.0024,Y=0.1366,Z=0.0519,W=-0.9893),Translation=(X=0,Y=0,Z=0),Scale3D=(X=1,Y=1,Z=1))))
    WeightMinimum=0
    Weight=@n2.Result
   )
   > &RigUnit_ModifyTransforms(
    ItemToModify=((Item=(Type=Bone,Name="MSTR_skirt1_r"),Transform=(Rotation=(X=0,Y=0,Z=-0.0872,W=0.9962),Translation=(X=0,Y=0,Z=0),Scale3D=(X=1,Y=1,Z=1))))
    WeightMinimum=0
    Weight=@n2.Result
   )
   > &RigUnit_ModifyTransforms(
    ItemToModify=((Item=(Type=Bone,Name="MSTR_skirt2_r"),Transform=(Rotation=(X=0,Y=0,Z=-0.0872,W=0.9962),Translation=(X=1.4153,Y=0,Z=0),Scale3D=(X=1,Y=1,Z=1))))
    WeightMinimum=0
    Weight=@n2.Result
   )
   > n3 = &RigUnit_SphericalPoseReader(
    ActiveRegionScaleFactors=(PositiveWidth=1,NegativeWidth=0.07,PositiveHeight=1,NegativeHeight=1)
    ActiveRegionSize=0.5
    Debug=(bDrawDebug=false,bDraw2D=False,bDrawLocalAxes=False,DebugScale=25,DebugSegments=20,DebugThickness=0.25)
    DriverAxis=(X=-1,Y=0,Z=0)
    DriverItem=(Type=Bone,Name="thigh_l")
    FalloffRegionScaleFactors=(PositiveWidth=1,NegativeWidth=0.9,PositiveHeight=1,NegativeHeight=1)
    FalloffSize=0.2
    FlipHeightScaling=False
    FlipWidthScaling=False
    OptionalParentItem=(Type=Bone,Name="pelvis")
    RotationOffset=(X=90,Y=0,Z=-8)
   )
   n4 = &RigVMFunction_AnimEvalRichCurve(
    Curve="(EditorCurveData=(Keys=((InterpMode=RCIM_Cubic,ArriveTangent=25.0927),(InterpMode=RCIM_Cubic,Time=0.0508,Value=0.1968,ArriveTangent=1.092,LeaveTangent=1.092),(InterpMode=RCIM_Cubic,TangentMode=RCTM_Break,Time=0.2657,Value=0.2902,ArriveTangent=0.4434,LeaveT ...(truncated)"
    SourceMinimum=0
    TargetMinimum=0
    Value=@n3.OutputParam
   )
   > &RigUnit_ModifyTransforms(
    ItemToModify=((Item=(Type=Bone,Name="MSTR_skirt_f_l"),Transform=(Rotation=(X=0.3895,Y=0.2778,Z=-0.4422,W=0.7586),Translation=(X=-4.3291,Y=-2.2575,Z=-3.4592),Scale3D=(X=1,Y=1,Z=1))))
    WeightMinimum=0
    Weight=@n4.Result
   )
   > &RigUnit_ModifyTransforms(
    ItemToModify=((Item=(Type=Bone,Name="MSTR_skirt_d_l"),Transform=(Rotation=(X=-0.2503,Y=-0.0448,Z=0.184,W=-0.9495),Translation=(X=3.4144,Y=-1.4177,Z=-2.804),Scale3D=(X=1,Y=1,Z=1))))
    WeightMinimum=0
    Weight=@n4.Result
   )
   > &RigUnit_ModifyTransforms(
    ItemToModify=((Item=(Type=Bone,Name="MSTR_skirt_b_l"),Transform=(Rotation=(X=0,Y=0,Z=-0.0872,W=0.9962),Translation=(X=0,Y=0,Z=0),Scale3D=(X=1,Y=1,Z=1))))
    WeightMinimum=0
    Weight=@n4.Result
   )
   > &RigUnit_ModifyTransforms(
    ItemToModify=((Item=(Type=Bone,Name="MSTR_skirt1_l"),Transform=(Rotation=(X=0,Y=0,Z=-0.284,W=0.9588),Translation=(X=0,Y=0,Z=0),Scale3D=(X=1,Y=1,Z=1))))
    WeightMinimum=0
    Weight=@n4.Result
   )
   > &RigUnit_ModifyTransforms(
    ItemToModify=((Item=(Type=Bone,Name="MSTR_skirt3_l"),Transform=(Rotation=(X=0,Y=0,Z=-0.1305,W=0.9914),Translation=(X=0,Y=0,Z=0),Scale3D=(X=1,Y=1,Z=1))))
    WeightMinimum=0
    Weight=@n4.Result
   )
   > &RigUnit_ModifyTransforms(
    ItemToModify=((Item=(Type=Bone,Name="MSTR_skirt2_l"),Transform=(Rotation=(X=0,Y=0,Z=-0.2164,W=0.9763),Translation=(X=0,Y=0,Z=0),Scale3D=(X=1,Y=1,Z=1))))
    WeightMinimum=0
    Weight=@n4.Result
   )
   n5 = &RigVMFunction_MathDoubleMul(A=@n4.Result, B=@n2.Result)
   n6 = &RigUnit_GetRelativeTransformForItem(Child=(Type=Bone,Name="pelvis"), Parent=(Type=Bone,Name="root"), bChildInitial=False, bParentInitial=False)
   > &RigUnit_ModifyTransforms(
    ItemToModify=((Item=(Type=Bone,Name="skirt1_b"),Transform=(Rotation=(X=0.0958,Y=0,Z=0,W=0.9954),Translation=(X=0,Y=0,Z=0),Scale3D=(X=1,Y=1,Z=1))))
    WeightMinimum=0
    Weight=&RigVMFunction_MathDoubleMul(
     A=@n5.Result
     B=&RigVMFunction_MathDoubleRemap(SourceMaximum=46, SourceMinimum=84, TargetMaximum=1, TargetMinimum=0, Value=@n6.RelativeTransform.Translation.Z).Result
    ).Result
   )
   n7 = &RigUnit_GetRelativeTransformForItem(Child=(Type=Bone,Name="spine_05"), Parent=(Type=Bone,Name="pelvis"), bChildInitial=False, bParentInitial=False)
   n8 = &RigVMFunction_MathDoubleMul(
    B=&RigVMFunction_MathDoubleRemap(
     SourceMaximum=43
     SourceMinimum=20
     TargetMaximum=1
     TargetMinimum=0
     Value=&RigVMFunction_MathQuaternionToEuler(RotationOrder=XYZ, Value=@n7.RelativeTransform.Rotation).Result.X
    ).Result
    A=@n4.Result
   )
   > &RigUnit_ModifyTransforms(
    ItemToModify="((Item=(Type=Bone,Name=\"MSTR_skirt_f_l\"),Transform=(Rotation=(X=-0.397,Y=0.0999,Z=0.2081,W=0.8883),Translation=(X=3.7444,Y=-0.444,Z=2.1541),Scale3D=(X=1,Y=1,Z=1))),(Item=(Type=Bone,Name=\"MSTR_skirt_f_r\"),Transform=(Rotation=(X=0,Y=0,Z=0,W=1),Translation=(X ...(truncated)"
    WeightMinimum=0
    Weight=@n8.Result
   )
   > &RigUnit_ModifyTransforms(
    ItemToModify="((Item=(Type=Bone,Name=\"MSTR_skirt2_l\"),Transform=(Rotation=(X=0,Y=0,Z=0.2164,W=-0.9763),Translation=(X=0.239,Y=-2.6651,Z=0),Scale3D=(X=1,Y=1,Z=1))),(Item=(Type=Bone,Name=\"MSTR_skirt1_l\"),Transform=(Rotation=(X=0,Y=0,Z=0,W=-1),Translation=(X=0,Y=-2.4255,Z= ...(truncated)"
    WeightMinimum=0
    Weight=@n8.Result
   )
   > n9 = &RigUnit_SphericalPoseReader(
    ActiveRegionScaleFactors=(PositiveWidth=1,NegativeWidth=1,PositiveHeight=0.8,NegativeHeight=1)
    ActiveRegionSize=0.1
    Debug=(bDrawDebug=true,bDraw2D=False,bDrawLocalAxes=False,DebugScale=25,DebugSegments=20,DebugThickness=0.25)
    DriverAxis=(X=0,Y=1.5,Z=0)
    DriverItem=(Type=Bone,Name="head")
    FalloffRegionScaleFactors=(PositiveWidth=1,NegativeWidth=1,PositiveHeight=1,NegativeHeight=1)
    FalloffSize=0.2
    FlipHeightScaling=False
    FlipWidthScaling=False
    OptionalParentItem=(Type=Bone,Name="spine_05")
    RotationOffset=(X=20,Y=-100,Z=0)
   )
   > &RigUnit_ModifyTransforms(
    ItemToModify=((Item=(Type=Bone,Name="MSTR_skirt_f_r"),Transform=(Rotation=(X=0,Y=0,Z=-0.1305,W=-0.9914),Translation=(X=1.2219,Y=-0.541,Z=2.7857),Scale3D=(X=1,Y=1,Z=1))))
    WeightMinimum=0
    Weight=&RigVMFunction_MathDoubleMul(B=@n5.Result, A=@n9.OutputParam).Result
   )
   > n10 = &RigUnit_SphericalPoseReader(
    ActiveRegionScaleFactors=(PositiveWidth=1,NegativeWidth=0.1,PositiveHeight=1,NegativeHeight=1)
    ActiveRegionSize=0.5
    Debug=(bDrawDebug=false,bDraw2D=False,bDrawLocalAxes=False,DebugScale=25,DebugSegments=20,DebugThickness=0.25)
    DriverAxis=(X=1,Y=0,Z=0)
    DriverItem=(Type=Bone,Name="thigh_r")
    FalloffRegionScaleFactors=(PositiveWidth=1,NegativeWidth=0.65,PositiveHeight=1,NegativeHeight=1)
    FalloffSize=0.2
    FlipHeightScaling=false
    FlipWidthScaling=false
    OptionalParentItem=(Type=Bone,Name="pelvis")
    RotationOffset=(X=25,Y=10,Z=90)
   )
   > &RigUnit_ModifyTransforms(
    ItemToModify="((Item=(Type=Bone,Name=\"MSTR_skirt_f_r\"),Transform=(Rotation=(X=-0.1931,Y=0.2423,Z=0.4492,W=-0.838),Translation=(X=0,Y=0,Z=0),Scale3D=(X=1,Y=1,Z=1))),(Item=(Type=Bone,Name=\"MSTR_skirt_d_r\"),Transform=(Rotation=(X=-0.0267,Y=-0.1475,Z=-0.2035,W=0.9675),Trans ...(truncated)"
    WeightMinimum=0
    Weight=@n10.OutputParam
   )
   > n11 = &RigUnit_SphericalPoseReader(
    ActiveRegionScaleFactors=(PositiveWidth=1,NegativeWidth=0.1,PositiveHeight=1,NegativeHeight=1)
    ActiveRegionSize=0.5
    Debug=(bDrawDebug=false,bDraw2D=False,bDrawLocalAxes=False,DebugScale=25,DebugSegments=20,DebugThickness=0.25)
    DriverAxis=(X=-1,Y=0,Z=0)
    DriverItem=(Type=Bone,Name="thigh_l")
    FalloffRegionScaleFactors=(PositiveWidth=1,NegativeWidth=0.3,PositiveHeight=1,NegativeHeight=1)
    FalloffSize=0.2
    FlipHeightScaling=false
    FlipWidthScaling=false
    OptionalParentItem=(Type=Bone,Name="pelvis")
    RotationOffset=(X=165,Y=-25,Z=-90)
   )
   > &RigUnit_ModifyTransforms(
    ItemToModify=((Item=(Type=Bone,Name="MSTR_skirt_f_l"),Transform=(Rotation=(X=0.0038,Y=0.0435,Z=-0.0871,W=0.9952),Translation=(X=0,Y=-3.7654,Z=-5.3776),Scale3D=(X=1,Y=1,Z=1))))
    WeightMinimum=0
    Weight=@n11.OutputParam
   )
   > Get Delta Ref_4 = RigVMFunctionReferenceNode(Item=(Type=Bone,Name="calf_l"))
   > Get Delta Ref_1 = RigVMFunctionReferenceNode(Item=(Type=Bone,Name="calf_r"))
   n12 = &RigVMFunction_MathDoubleRemap(SourceMaximum=133, SourceMinimum=25, TargetMaximum=1, TargetMinimum=0, Value=@"Get Delta Ref_1".Result.Rotation.Yaw)
   n13 = &RigVMFunction_MathDoubleRemap(
    SourceMaximum=25
    SourceMinimum=40
    TargetMaximum=1
    TargetMinimum=0
    Value=&RigUnit_GetRelativeTransformForItem(Child=(Type=Bone,Name="pelvis"), Parent=(Type=Bone,Name="root"), bChildInitial=False, bParentInitial=False).RelativeTransform.Translation.Z
   )
   n14 = &RigVMFunction_MathDoubleMul(A=@n12.Result, B=@n13.Result)
   > &RigUnit_ModifyTransforms(
    ItemToModify=((Item=(Type=Bone,Name="MSTR_skirt_f_r"),Transform=(Rotation=(X=-0.3499,Y=-0.1961,Z=0.3056,W=-0.8636),Translation=(X=-2.6763,Y=3.3165,Z=-0.8659),Scale3D=(X=1,Y=1,Z=1))))
    WeightMinimum=0
    Weight=@n14.Result
   )
   > &RigUnit_ModifyTransforms(
    ItemToModify=((Item=(Type=Bone,Name="MSTR_skirt_d_r"),Transform=(Rotation=(X=0.3127,Y=-0.0316,Z=-0.6465,W=0.6952),Translation=(X=-2.2037,Y=-4.4058,Z=-2.248),Scale3D=(X=1,Y=1,Z=1))))
    WeightMinimum=0
    Weight=@n14.Result
   )
   > &RigUnit_ModifyTransforms(
    ItemToModify=((Item=(Type=Bone,Name="MSTR_skirt1_b_r"),Transform=(Rotation=(X=0.3667,Y=-0.1835,Z=0.5888,W=-0.6965),Translation=(X=0,Y=0,Z=0),Scale3D=(X=1,Y=1,Z=1))))
    WeightMinimum=0
    Weight=@n14.Result
   )
   > &RigUnit_ModifyTransforms(
    ItemToModify=((Item=(Type=Bone,Name="MSTR_skirt_f_l"),Transform=(Rotation=(X=-0.0581,Y=-0.1016,Z=-0.0506,W=0.9918),Translation=(X=1.2881,Y=-4.1965,Z=0.5341),Scale3D=(X=1,Y=1,Z=1))))
    WeightMinimum=0
    Weight=@n14.Result
   )
   > &RigUnit_ModifyTransforms(
    ItemToModify=((Item=(Type=Bone,Name="MSTR_skirt_d_l"),Transform=(Rotation=(X=-0.116,Y=-0.2468,Z=-0.0259,W=0.9618),Translation=(X=0,Y=0,Z=0),Scale3D=(X=1,Y=1,Z=1))))
    WeightMinimum=0
    Weight=@n14.Result
   )
   > &RigUnit_ModifyTransforms(
    ItemToModify=((Item=(Type=Bone,Name="MSTR_skirt_b_l"),Transform=(Rotation=(X=-0.2386,Y=-0.2139,Z=-0.5135,W=0.796),Translation=(X=0,Y=0,Z=0),Scale3D=(X=1,Y=1,Z=1))))
    WeightMinimum=0
    Weight=@n14.Result
   )
   > &RigUnit_ModifyTransforms(
    ItemToModify=((Item=(Type=Bone,Name="MSTR_skirt_b"),Transform=(Rotation=(X=-0.2655,Y=-0.208,Z=0.3411,W=0.8774),Translation=(X=0,Y=-0.0571,Z=-0.6384),Scale3D=(X=1,Y=1,Z=1))))
    WeightMinimum=0
    Weight=@n14.Result
   )
   > &RigUnit_ModifyTransforms(
    ItemToModify=((Item=(Type=Bone,Name="MSTR_skirt1_l"),Transform=(Rotation=(X=0,Y=0,Z=0.284,W=-0.9588),Translation=(X=0.1342,Y=-1.1414,Z=1.3548),Scale3D=(X=1,Y=1,Z=1))))
    WeightMinimum=0
    Weight=@n14.Result
   )
   Multiply = &RigVMFunction_MathDoubleMul(
    A=@n12.Result
    B=&RigVMFunction_MathDoubleRemap(SourceMaximum=133, SourceMinimum=25, TargetMaximum=1, TargetMinimum=0, Value=@"Get Delta Ref_4".Result.Rotation.Yaw).Result
   )
   n15 = &RigVMFunction_MathDoubleMul(A=@Multiply.Result, B=@n13.Result)
   > &RigUnit_ModifyTransforms(
    ItemToModify=((Item=(Type=Bone,Name="MSTR_skirt_f_r"),Transform=(Rotation=(X=0.4565,Y=0.2551,Z=-0.3592,W=-0.773),Translation=(X=8.5434,Y=-2.719,Z=-4.5074),Scale3D=(X=1,Y=1,Z=1))))
    WeightMinimum=0
    Weight=@n15.Result
   )
   > &RigUnit_ModifyTransforms(
    ItemToModify=((Item=(Type=Bone,Name="MSTR_skirt_d_r"),Transform=(Rotation=(X=0.2002,Y=-0.0633,Z=0.4729,W=0.8557),Translation=(X=0,Y=0,Z=0),Scale3D=(X=1,Y=1,Z=1))))
    WeightMinimum=0
    Weight=@n15.Result
   )
   > &RigUnit_ModifyTransforms(
    ItemToModify=((Item=(Type=Bone,Name="MSTR_skirt1_b_r"),Transform=(Rotation=(X=0.1457,Y=-0.1041,Z=0.139,W=0.974),Translation=(X=0,Y=0,Z=0),Scale3D=(X=1,Y=1,Z=1))))
    WeightMinimum=0
    Weight=@n15.Result
   )
   > &RigUnit_ModifyTransforms(
    ItemToModify=((Item=(Type=Bone,Name="MSTR_skirt_f_l"),Transform=(Rotation=(X=-0.0844,Y=0.1518,Z=0.0161,W=0.9847),Translation=(X=0,Y=0,Z=1),Scale3D=(X=1,Y=1,Z=1))))
    WeightMinimum=0
    Weight=@n15.Result
   )
   > &RigUnit_ModifyTransforms(
    ItemToModify=((Item=(Type=Bone,Name="MSTR_skirt_d_l"),Transform=(Rotation=(X=0.4397,Y=-0.4745,Z=0.2767,W=-0.7107),Translation=(X=0,Y=0,Z=0),Scale3D=(X=1,Y=1,Z=1))))
    WeightMinimum=0
    Weight=@n15.Result
   )
   > &RigUnit_ModifyTransforms(
    ItemToModify=((Item=(Type=Bone,Name="MSTR_skirt_b_l"),Transform=(Rotation=(X=-0.0113,Y=-0.2586,Z=0.0421,W=-0.965),Translation=(X=0,Y=0,Z=0),Scale3D=(X=1,Y=1,Z=1))))
    WeightMinimum=0
    Weight=@n15.Result
   )
   > &RigUnit_ModifyTransforms(
    ItemToModify=((Item=(Type=Bone,Name="MSTR_skirt_b"),Transform=(Rotation=(X=0.0187,Y=-0.2733,Z=0.5703,W=-0.7744),Translation=(X=0,Y=0,Z=0),Scale3D=(X=1,Y=1,Z=1))))
    WeightMinimum=0
    Weight=@n15.Result
   )
   Get Delta Ref_5 = RigVMFunctionReferenceNode(Item=(Type=Bone,Name="foot_l"), ExecuteContext=@RigVMFunction_Sequence.B)
   > &RigUnit_ModifyTransforms(
    ItemToModify=((Item=(Type=Bone,Name="cor_foot_l"),Transform=(Rotation=(X=0,Y=0,Z=0,W=-1),Translation=(X=6.5774,Y=-13.5693,Z=1.0704),Scale3D=(X=1,Y=1,Z=1))))
    WeightMinimum=0
    Weight=&RigVMFunction_MathDoubleRemap(SourceMaximum=60, SourceMinimum=10, TargetMaximum=1, TargetMinimum=0, Value=@"Get Delta Ref_5".Result.Rotation.Yaw).Result
   )
   > &RigUnit_ModifyTransforms(
    ItemToModify=((Item=(Type=Bone,Name="cor_foot_l"),Transform=(Rotation=(X=0,Y=0,Z=0,W=1),Translation=(X=3.0642,Y=-0.4381,Z=0),Scale3D=(X=1,Y=1,Z=1))))
    WeightMinimum=0
    Weight=&RigVMFunction_MathDoubleRemap(SourceMaximum=-38, SourceMinimum=-13, TargetMaximum=1, TargetMinimum=0, Value=@"Get Delta Ref_5".Result.Rotation.Yaw).Result
   )
   > Get Delta Ref_2 = RigVMFunctionReferenceNode(Item=(Type=Bone,Name="foot_r"))
   > &RigUnit_ModifyTransforms(
    ItemToModify=((Item=(Type=Bone,Name="cor_foot_r"),Transform=(Rotation=(X=0,Y=0,Z=0,W=1),Translation=(X=10.4892,Y=-12.452,Z=2.146),Scale3D=(X=1,Y=1,Z=1))))
    WeightMinimum=0
    Weight=&RigVMFunction_MathDoubleRemap(SourceMaximum=60, SourceMinimum=10, TargetMaximum=1, TargetMinimum=0, Value=@"Get Delta Ref_2".Result.Rotation.Yaw).Result
   )
   > &RigUnit_ModifyTransforms(
    ItemToModify=((Item=(Type=Bone,Name="cor_foot_r"),Transform=(Rotation=(X=0,Y=0,Z=0,W=1),Translation=(X=5.045,Y=-1.4961,Z=0),Scale3D=(X=1,Y=1,Z=1))))
    WeightMinimum=0
    Weight=&RigVMFunction_MathDoubleRemap(SourceMaximum=-40, SourceMinimum=-13, TargetMaximum=1, TargetMinimum=0, Value=@"Get Delta Ref_2".Result.Rotation.Yaw).Result
   )
   RigVMFunction_MathQuaternionToEuler = &RigVMFunction_MathQuaternionToEuler(
    RotationOrder=XYZ
    Value=&RigUnit_GetRelativeTransformForItem(Child=(Type=Bone,Name="clavicle_l"), Parent=(Type=Bone,Name="spine_05"), bChildInitial=False, bParentInitial=False).RelativeTransform.Rotation
   )
   Remap = &RigVMFunction_MathDoubleRemap(SourceMaximum=60, SourceMinimum=90, TargetMaximum=1, TargetMinimum=0, Value=@RigVMFunction_MathQuaternionToEuler.Result.Y)
   n16 = &RigVMFunction_MathQuaternionToEuler(
    RotationOrder=XYZ
    Value=&RigUnit_GetRelativeTransformForItem(Child=(Type=Bone,Name="clavicle_r"), Parent=(Type=Bone,Name="spine_05"), bChildInitial=False, bParentInitial=False).RelativeTransform.Rotation
   )
   &RigUnit_ModifyTransforms(
    ItemToModify="((Item=(Type=Bone,Name=\"MSTR_skirt_b\"),Transform=(Rotation=(X=-0.0872,Y=0,Z=0,W=0.9962),Translation=(X=0,Y=0,Z=0),Scale3D=(X=1,Y=1,Z=1))),(Item=(Type=Bone,Name=\"MSTR_sholder_l\"),Transform=(Rotation=(X=0,Y=0,Z=0,W=-1),Translation=(X=0,Y=0,Z=0),Scale3D=(X=1, ...(truncated)"
    WeightMinimum=0
    Weight=&RigVMFunction_MathDoubleMul(
     A=@Remap.Result
     B=&RigVMFunction_MathDoubleRemap(SourceMaximum=60, SourceMinimum=90, TargetMaximum=1, TargetMinimum=0, Value=@n16.Result.Y).Result
    ).Result
    ExecutePin=@RigVMFunction_Sequence.C
   )
   RigVMRerouteNode(Value=())
  ]}
]
hierarchy [
 root {loc=<0 0 0>, rot=<0 0 0 1>}
 pelvis {parent=root, loc=<0 -0.5301 98.0571>, rot=<0.0127 0.707 -0.0127 -0.707>}
 spine_01 {parent=pelvis, loc=<6.1208 -0.9328 0>, rot=<0 0 0.034 -0.9994>}
 spine_02 {parent=spine_01, loc=<6.5997 -0.2095 0>, rot=<0 0 0.0238 -0.9997>}
 spine_03 {parent=spine_02, loc=<6.4914 -0.1237 0>, rot=<0 0 0.0149 -0.9999>}
 spine_04 {parent=spine_03, loc=<6.9598 0.0794 0>, rot=<0 0 -0.0132 -0.9999>}
 spine_05 {parent=spine_04, loc=<6.9987 0 0>, rot=<0 0 -0.0015 -1>}
 neck_01 {parent=spine_05, loc=<11.3138 0.1529 0>, rot=<0 0 -0.2003 -0.9797>}
 neck_02 {parent=neck_01, loc=<4.5627 0 0>, rot=<0 0 0.0504 -0.9987>}
 head {parent=neck_02, loc=<5.3914 0 0>, rot=<0 0 0.115 -0.9934>}
 MSTR_hair1_f {parent=head, loc=<8.7939 9.44 -2.5372>, rot=<0.0029 -0.7071 0.7071 -0.0029>}
 hair1_f {parent=MSTR_hair1_f, loc=<0.0047 0 0>, rot=<0 0 0 1>}
 MSTR_hair2_f {parent=head, loc=<8.6402 10.2092 0.4359>, rot=<0.0029 -0.7071 0.7071 -0.0029>}
 hair2_f {parent=MSTR_hair2_f, loc=<0 0 0>, rot=<0 0 0 1>}
 MSTR_hair_f2_l {parent=head, loc=<2.8474 2.9717 -8.4712>, rot=<0.0511 0.5865 0.8083 0.0021>}
 hair1_f2_l {parent=MSTR_hair_f2_l, loc=<0 0 0>, rot=<0 0 0 1>}
 MSTR_hair_f1_l {parent=head, loc=<3.8987 6.4532 -6.8944>, rot=<0.0323 0.4181 0.9077 0.0111>}
 hair1_f1_l {parent=MSTR_hair_f1_l, loc=<0 0 0>, rot=<0 0 0 1>}
 MSTR_hair_f1_r {parent=head, loc=<2.5298 6.5768 6.6467>, rot=<0.0109 -0.4376 0.8982 -0.039>}
 hair1_f1_r {parent=MSTR_hair_f1_r, loc=<0 0 0>, rot=<0 0 0 1>}
 MSTR_hair_f2_r {parent=head, loc=<2.9489 3.6117 8.5926>, rot=<0.0034 -0.5698 0.8208 -0.0411>}
 hair1_f2_r {parent=MSTR_hair_f2_r, loc=<0 0 0>, rot=<0 0 0 1>}
 MSTR_hair2_f3_r {parent=head, loc=<-5.6219 -0.3064 5.8272>, rot=<0.0405 -0.0163 0.9269 0.3729>}
 hair2_f3_r {parent=MSTR_hair2_f3_r, loc=<0 0 0>, rot=<0 0 0 1>}
 hair3_f3_r {parent=hair2_f3_r, loc=<3.7949 2.4053 0>, rot=<0 0 0 1>}
 hair4_f3_r {parent=hair3_f3_r, loc=<5.4887 -0.1426 0.012>, rot=<0 0 0 -1>}
 hair5_f3_r {parent=hair4_f3_r, loc=<5.2985 0.3147 -0.1115>, rot=<0 0 0 1>}
 MSTR_hair2_f3_l {parent=head, loc=<-5.6179 -0.3107 -6.0865>, rot=<0.0202 -0.0081 0.9275 0.3731>}
 hair2_f3_l {parent=MSTR_hair2_f3_l, loc=<0 0 0>, rot=<0 0 0 1>}
 hair3_f3_l {parent=hair2_f3_l, loc=<3.7954 2.4053 -0.0226>, rot=<0 0 0 1>}
 hair4_f3_l {parent=hair3_f3_l, loc=<5.472 -0.1643 0.0959>, rot=<0 0 0 1>}
 hair5_f3_l {parent=hair4_f3_l, loc=<5.3183 0.3362 -0.1231>, rot=<0 0 0 1>}
 hairMaster {parent=head, loc=<7.3492 -6.6111 0>, rot=<0.0051 -1 -0.0007 -0.0005>}
 base_backHair_M_01 {parent=hairMaster, loc=<8.073 -0.813 0>, rot=<0 0 0.1777 -0.9841>}
 dyn_backHair_M_01 {parent=base_backHair_M_01, loc=<0 0 0>, rot=<0 0 0 1>}
 dyn_backHair_M_02 {parent=dyn_backHair_M_01, loc=<7.2872 1.2981 0.0062>, rot=<0 0 -0.0941 -0.9956>}
 dyn_backHair_M_03 {parent=dyn_backHair_M_02, loc=<6.4877 -1.0417 0>, rot=<0 0 0 1>}
 Dyn_hair_top_M_01 {parent=base_backHair_M_01, loc=<0.183 -0.4479 0.2377>, rot=<0.6858 -0.06 0.0287 -0.7248>}
 Dyn_hair_top_M_02 {parent=Dyn_hair_top_M_01, loc=<7.9601 0 0>, rot=<-0.0177 0.0106 0.003 -0.9998>}
 Dyn_hair_top_M_03 {parent=Dyn_hair_top_M_02, loc=<4.2768 0 0>, rot=<0 0 0 1>}
 Star {parent=Dyn_hair_top_M_01, loc=<3.7528 0.0963 -0.2417>, rot=<-0.4714 0.5202 0.5024 -0.5047>}
 base_backHair_R_01 {parent=hairMaster, loc=<8.073 1.4486 -7.5>, rot=<0 0 0.2349 -0.972>}
 dyn_backHair_R_01 {parent=base_backHair_R_01, loc=<0 0 0>, rot=<0 0 0 1>}
 dyn_backHair_R_02 {parent=dyn_backHair_R_01, loc=<7.5693 1.5002 0.0063>, rot=<0 0 -0.1261 -0.992>}
 dyn_backHair_R_03 {parent=dyn_backHair_R_02, loc=<6.7268 -1.5764 0>, rot=<0 0 0 1>}
 base_backHair_L_01 {parent=hairMaster, loc=<8.073 1.4486 7.5>, rot=<0 0 0.2162 -0.9763>}
 dyn_backHair_L_01 {parent=base_backHair_L_01, loc=<0 0 0>, rot=<0 0 0 1>}
 dyn_backHair_L_02 {parent=dyn_backHair_L_01, loc=<7.6466 1.1695 0.0056>, rot=<0 0 -0.0963 -0.9954>}
 dyn_backHair_L_03 {parent=dyn_backHair_L_02, loc=<6.747 -1.4012 0>, rot=<0 0 0 1>}
 base_hair_Front_R_01 {parent=head, loc=<0.6697 -0.0576 5.9219>, rot=<0.0039 -0.0238 0.9986 0.0468>}
 dyn_hair_Front_R_01 {parent=base_hair_Front_R_01, loc=<0 0 0>, rot=<0 0 0 1>}
 dyn_hair_Front_R_02 {parent=dyn_hair_Front_R_01, loc=<7.6629 -0.403 0>, rot=<0.0148 0.015 0.151 -0.9883>}
 dyn_hair_Front_R_03 {parent=dyn_hair_Front_R_02, loc=<5.8745 0 0>, rot=<0.0014 0.0017 0.1263 -0.992>}
 dyn_hair_Front_R_04 {parent=dyn_hair_Front_R_03, loc=<5.6396 0 0>, rot=<-0.0112 -0.0138 -0.04 -0.999>}
 base_hair_Front_L_01 {parent=head, loc=<1.0144 -0.1667 -6.3229>, rot=<-0.0141 -0.0435 0.9985 0.029>}
 dyn_hair_Front_L_01 {parent=base_hair_Front_L_01, loc=<0 0 0>, rot=<0 0 0 1>}
 dyn_hair_Front_L_02 {parent=dyn_hair_Front_L_01, loc=<7.5565 0.406 0.3554>, rot=<-0.0047 -0.0058 0.329 -0.9443>}
 dyn_hair_Front_L_03 {parent=dyn_hair_Front_L_02, loc=<6.925 1.1575 0.0542>, rot=<-0.0033 -0.0174 0.0369 -0.9992>}
 dyn_hair_Front_L_04 {parent=dyn_hair_Front_L_03, loc=<5.6943 1.0534 0.0776>, rot=<0.0028 0.0068 -0.1488 -0.9888>}
 neck_01_mover {parent=neck_01, loc=<0 0 0>, rot=<0 0 0 1>}
 clavicle_l {parent=spine_05, loc=<5.4831 -0.1053 -3.2946>, rot=<0.0465 -0.7289 -0.009 -0.6829>}
 upperarm_l {parent=clavicle_l, loc=<9.8658 -1.2136 -0.4105>, rot=<0.0164 -0.3245 -0.0269 -0.9454>}
 lowerarm_l {parent=upperarm_l, loc=<26.465 -0.8345 0.259>, rot=<0.0036 0.0161 -0.2298 -0.9731>}
 hand_l {parent=lowerarm_l, loc=<22.7355 -0.3203 -0.6105>, rot=<-0.6784 -0.0223 0.0735 -0.7307>}
 thumb_01_l {parent=hand_l, loc=<1.3118 -0.3191 -2.459>, rot=<0.6546 -0.4042 0.0664 -0.6354>}
 thumb_02_l {parent=thumb_01_l, loc=<5.9336 -0.0448 0.0401>, rot=<-0.0037 0.0213 0.298 -0.9543>}
 thumb_03_l {parent=thumb_02_l, loc=<2.2223 0.3682 -0.0907>, rot=<0.0118 -0.013 0.0066 -0.9998>}
 index_metacarpal_l {parent=hand_l, loc=<1.6535 0.0183 -2.96>, rot=<0.0158 -0.0124 -0.0173 -0.9996>}
 index_01_l {parent=index_metacarpal_l, loc=<8.3008 0.0357 -0.0524>, rot=<-0.0044 0.0485 0.0924 -0.9945>}
 index_02_l {parent=index_01_l, loc=<3.3587 -0.3004 0.0452>, rot=<-0.002 0.0027 0.3225 -0.9466>}
 index_03_l {parent=index_02_l, loc=<2.5578 -0.123 -0.0252>, rot=<-0.002 0.0067 0.2558 -0.9667>}
 middle_metacarpal_l {parent=hand_l, loc=<2.1856 -0.1725 -0.5984>, rot=<0.006 0.0168 -0.0329 -0.9993>}
 middle_01_l {parent=middle_metacarpal_l, loc=<7.3275 -0.2447 0.0129>, rot=<-0.0208 0.0042 0.0858 -0.9961>}
 middle_02_l {parent=middle_01_l, loc=<4.0123 0.1401 0.1175>, rot=<-0.0218 0.0258 0.3512 -0.9357>}
 middle_03_l {parent=middle_02_l, loc=<2.6077 0.0534 0.1373>, rot=<0.0082 0.0188 0.1708 -0.9851>}
 ring_metacarpal_l {parent=hand_l, loc=<2.2074 -0.3625 1.3638>, rot=<-0.0284 0.0722 -0.0458 -0.9959>}
 ring_01_l {parent=ring_metacarpal_l, loc=<7.0295 -0.6197 -0.1446>, rot=<-0.0489 -0.0252 0.1892 -0.9804>}
 ring_02_l {parent=ring_01_l, loc=<3.6174 0.4647 -0.1027>, rot=<0 0 0.2806 -0.9598>}
 ring_03_l {parent=ring_02_l, loc=<2.4152 0.114 0.0199>, rot=<0 0 0.1189 -0.9929>}
 pinky_metacarpal_l {parent=hand_l, loc=<1.7616 -0.9753 3.3814>, rot=<-0.0397 0.0814 -0.069 -0.9935>}
 pinky_01_l {parent=pinky_metacarpal_l, loc=<7.1709 -0.4831 -0.1094>, rot=<-0.0654 -0.0035 0.2242 -0.9723>}
 pinky_02_l {parent=pinky_01_l, loc=<2.8274 0.0893 -0.1043>, rot=<0.0037 0.0011 0.2854 -0.9584>}
 pinky_03_l {parent=pinky_02_l, loc=<2.2046 0.0307 -0.0078>, rot=<-0.0002 0.0098 0.1445 -0.9895>}
 hand_attach_l {parent=hand_l, loc=<8.5 -3 0>, rot=<0.7071 0.7071 0 0>}
 weapon_l {parent=hand_l, loc=<-0.1974 -3.4111 0.3811>, rot=<0 0 0 1>}
 deform_wrist_dn_base_l {parent=hand_l, loc=<0 -1.5326 0>, rot=<0 0 0 -1>}
 deform_wrist_dn_l {parent=deform_wrist_dn_base_l, loc=<0 0 0>, rot=<0 0 0 1>}
 deform_wrist_up_base_l {parent=hand_l, loc=<0 2.3009 0>, rot=<0 0 0 -1>}
 deform_wrist_up_l {parent=deform_wrist_up_base_l, loc=<0 0 0>, rot=<0 0 0 1>}
 lowerarm_twist_02_l {parent=lowerarm_l, loc=<10.6707 0 0>, rot=<0 0 0 1>}
 lowerarm_twist_01_l {parent=lowerarm_l, loc=<18.316 0 0>, rot=<0 0 0 1>}
 rings_cor {parent=lowerarm_twist_01_l, loc=<2.915 -0.2399 -0.4147>, rot=<-0.6784 -0.0223 0.0735 -0.7307>}
 rings_cor1 {parent=lowerarm_twist_01_l, loc=<1.3032 -0.0009 -0.4074>, rot=<-0.6784 -0.0223 0.0735 -0.7307>}
 deform_elbow_in_base_l {parent=lowerarm_l, loc=<0.1834 3.6062 -0.2572>, rot=<0 0 0 -1>}
 deform_elbow_in_l {parent=deform_elbow_in_base_l, loc=<0 0 0>, rot=<0 0 0 1>}
 deform_elbow_out_base_l {parent=lowerarm_l, loc=<0 -4 0>, rot=<0 0 0 1>}
 deform_elbow_out_l {parent=deform_elbow_out_base_l, loc=<0 0 0>, rot=<0 0 0 1>}
 upperarm_twist_01_l {parent=upperarm_l, loc=<7.2135 -0.1134 0>, rot=<0 0 0 1>}
 deform_upperarm_bk_base_l {parent=upperarm_twist_01_l, loc=<-6.4053 -4.3746 0.314>, rot=<0 0 0 -1>}
 deform_upperarm_bk_l {parent=deform_upperarm_bk_base_l, loc=<0 0 0>, rot=<0 0 0 1>}
 deform_upperarm_fr_base_l {parent=upperarm_twist_01_l, loc=<-5.6312 5.092 0.3354>, rot=<0 0 0 -1>}
 deform_upperarm_fr_l {parent=deform_upperarm_fr_base_l, loc=<0 0 0>, rot=<0 0 0 1>}
 deform_upperarm_up_base_l {parent=upperarm_twist_01_l, loc=<-8.891 0.4825 3.9517>, rot=<0 0 0 -1>}
 deform_upperarm_up_l {parent=deform_upperarm_up_base_l, loc=<0 0 0>, rot=<0 0 0 1>}
 upperarm_twist_02_l {parent=upperarm_l, loc=<19.7998 -0.1134 0>, rot=<0 0 0 1>}
 MSTR_sholder_l {parent=clavicle_l, loc=<9.1137 -0.8088 4.9652>, rot=<0.6833 0.1866 0.1515 -0.6895>}
 sholder_l {parent=MSTR_sholder_l, loc=<0 0 0>, rot=<0 0 0 1>}
 hairClavicleCorr_l {parent=clavicle_l, loc=<7.4084 -1.3384 4.6551>, rot=<-0.0012 0.0335 0.0362 -0.9988>}
 clavicle_r {parent=spine_05, loc=<5.4831 -0.1053 3.2946>, rot=<0.7289 0.0465 -0.6829 0.009>}
 upperarm_r {parent=clavicle_r, loc=<-9.8658 1.2136 0.4105>, rot=<0.0164 -0.3245 -0.0269 -0.9454>}
 lowerarm_r {parent=upperarm_r, loc=<-26.465 0.8345 -0.259>, rot=<0.0036 0.0161 -0.2298 -0.9731>}
 hand_r {parent=lowerarm_r, loc=<-22.7355 0.3203 0.6106>, rot=<-0.6784 -0.0223 0.0735 -0.7307>}
 thumb_01_r {parent=hand_r, loc=<-1.3118 0.3191 2.459>, rot=<0.6546 -0.4042 0.0664 -0.6354>}
 thumb_02_r {parent=thumb_01_r, loc=<-5.9336 0.0448 -0.0401>, rot=<-0.0037 0.0213 0.298 -0.9543>}
 thumb_03_r {parent=thumb_02_r, loc=<-2.2223 -0.3682 0.0907>, rot=<0.0118 -0.013 0.0066 -0.9998>}
 index_metacarpal_r {parent=hand_r, loc=<-1.6535 -0.0183 2.96>, rot=<0.0158 -0.0124 -0.0173 -0.9996>}
 index_01_r {parent=index_metacarpal_r, loc=<-8.3008 -0.0357 0.0524>, rot=<-0.0044 0.0485 0.0924 -0.9945>}
 index_02_r {parent=index_01_r, loc=<-3.3587 0.3004 -0.0452>, rot=<-0.002 0.0027 0.3225 -0.9466>}
 index_03_r {parent=index_02_r, loc=<-2.5578 0.123 0.0252>, rot=<-0.002 0.0067 0.2558 -0.9667>}
 middle_metacarpal_r {parent=hand_r, loc=<-2.1856 0.1725 0.5984>, rot=<0.006 0.0168 -0.0329 -0.9993>}
 middle_01_r {parent=middle_metacarpal_r, loc=<-7.3275 0.2447 -0.0129>, rot=<-0.0208 0.0042 0.0858 -0.9961>}
 middle_02_r {parent=middle_01_r, loc=<-4.0123 -0.1401 -0.1175>, rot=<-0.0218 0.0258 0.3512 -0.9357>}
 middle_03_r {parent=middle_02_r, loc=<-2.6077 -0.0534 -0.1373>, rot=<0.0082 0.0188 0.1708 -0.9851>}
 ring_metacarpal_r {parent=hand_r, loc=<-2.2074 0.3625 -1.3638>, rot=<-0.0284 0.0722 -0.0458 -0.9959>}
 ring_01_r {parent=ring_metacarpal_r, loc=<-7.0295 0.6197 0.1446>, rot=<-0.0489 -0.0252 0.1892 -0.9804>}
 ring_02_r {parent=ring_01_r, loc=<-3.6174 -0.4647 0.1027>, rot=<0 0 0.2806 -0.9598>}
 ring_03_r {parent=ring_02_r, loc=<-2.4152 -0.114 -0.0199>, rot=<0 0 0.1189 -0.9929>}
 pinky_metacarpal_r {parent=hand_r, loc=<-1.7616 0.9753 -3.3814>, rot=<-0.0397 0.0814 -0.069 -0.9935>}
 pinky_01_r {parent=pinky_metacarpal_r, loc=<-7.1709 0.4831 0.1094>, rot=<-0.0654 -0.0035 0.2242 -0.9723>}
 pinky_02_r {parent=pinky_01_r, loc=<-2.8274 -0.0893 0.1043>, rot=<0.0037 0.0011 0.2854 -0.9584>}
 pinky_03_r {parent=pinky_02_r, loc=<-2.2046 -0.0307 0.0078>, rot=<-0.0002 0.0098 0.1445 -0.9895>}
 weapon_r {parent=hand_r, loc=<0.1974 3.4111 -0.3811>, rot=<0 0 0 1>}
 hand_attach_r {parent=hand_r, loc=<-8.5001 2.9999 0>, rot=<0 0 -0.7071 -0.7071>}
 deform_wrist_dn_base_r {parent=hand_r, loc=<-0.0001 1.5328 0>, rot=<0 0 0 -1>}
 deform_wrist_dn_r {parent=deform_wrist_dn_base_r, loc=<0 0 0>, rot=<0 0 0 1>}
 deform_wrist_up_base_r {parent=hand_r, loc=<-0.0002 -2.3007 -0.0001>, rot=<0 0 0 -1>}
 deform_wrist_up_r {parent=deform_wrist_up_base_r, loc=<0 0 0>, rot=<0 0 0 1>}
 lowerarm_twist_02_r {parent=lowerarm_r, loc=<-10.6707 0 0>, rot=<0 0 0 -1>}
 lowerarm_twist_01_r {parent=lowerarm_r, loc=<-18.316 0 0>, rot=<0 0 0 -1>}
 deform_elbow_in_base_r {parent=lowerarm_r, loc=<-0.1834 -3.6062 0.2572>, rot=<0 0 0 -1>}
 deform_elbow_in_r {parent=deform_elbow_in_base_r, loc=<0 0 0>, rot=<0 0 0 1>}
 deform_elbow_out_base_r {parent=lowerarm_r, loc=<0 4 -0.0001>, rot=<0 0 0 -1>}
 deform_elbow_out_r {parent=deform_elbow_out_base_r, loc=<0 0 0>, rot=<0 0 0 1>}
 upperarm_twist_01_r {parent=upperarm_r, loc=<-7.2135 0.1134 0>, rot=<0 0 0 -1>}
 deform_upperarm_bk_base_r {parent=upperarm_twist_01_r, loc=<6.4053 4.3746 -0.314>, rot=<0 0 0 -1>}
 deform_upperarm_bk_r {parent=deform_upperarm_bk_base_r, loc=<0 0 0>, rot=<0 0 0 1>}
 deform_upperarm_fr_base_r {parent=upperarm_twist_01_r, loc=<5.6312 -5.092 -0.3354>, rot=<0 0 0 -1>}
 deform_upperarm_fr_r {parent=deform_upperarm_fr_base_r, loc=<0 0 0>, rot=<0 0 0 1>}
 deform_upperarm_up_base_r {parent=upperarm_twist_01_r, loc=<8.891 -0.4825 -3.9517>, rot=<0 0 0 -1>}
 deform_upperarm_up_r {parent=deform_upperarm_up_base_r, loc=<0 0 0>, rot=<0 0 0 1>}
 upperarm_twist_02_r {parent=upperarm_r, loc=<-19.7998 0.1134 0>, rot=<0 0 0 -1>}
 MSTR_sholder_r {parent=clavicle_r, loc=<-9.1137 0.8088 -4.9656>, rot=<-0.1613 0.6897 0.6946 0.126>}
 sholder_r {parent=MSTR_sholder_r, loc=<0 0 0>, rot=<0 0 0 1>}
 hairClavicleCorr_r {parent=clavicle_r, loc=<-7.0476 0.8858 -4.9235>, rot=<0.9976 -0.0668 0.0195 -0.0016>}
 attach_backpack {parent=spine_05, loc=<0 -8.35 0>, rot=<-0.0093 -0.707 0.0093 -0.707>}
 Backpack_BR {parent=attach_backpack, type=Null, loc=<0 0 0>, rot=<0 0 0 1>}
 attach_cape {parent=spine_05, loc=<9.3216 -8.4501 0>, rot=<0.0616 -0.7044 -0.0616 -0.7044>}
 cape {parent=attach_cape, type=Null, loc=<0 0 0>, rot=<0 0 0 1>}
 MSTR_hair2_b {parent=spine_05, loc=<20.5775 -4.1732 0.3337>, rot=<0.1263 -0.992 0 0>}
 hair2_b {parent=MSTR_hair2_b, loc=<0 0 0>, rot=<0 0 0 1>}
 hair3_b {parent=hair2_b, loc=<5.4361 0.0029 0>, rot=<0 0 0.0274 -0.9996>}
 hair3_b_l {parent=hair3_b, loc=<5.2933 -0.913 5.0185>, rot=<0.0655 0.1451 0.0621 -0.9853>}
 hair4_b_l {parent=hair3_b_l, loc=<8.4119 0 0>, rot=<0 0 0 1>}
 hair3_b_r {parent=hair3_b, loc=<5.3132 -0.9309 -4.9464>, rot=<-0.1511 -0.1392 0.0745 -0.9758>}
 hair4_b_r {parent=hair3_b_r, loc=<8.4124 0 0>, rot=<0 0 0 1>}
 MSTR_hair_b_l {parent=hair2_b, loc=<6.9482 3.8421 7.1137>, rot=<0.8051 -0.5337 0.2157 0.143>}
 MSTR_hair_b_r {parent=hair2_b, loc=<6.9482 3.8421 -6.4464>, rot=<0.8051 -0.5337 -0.2157 -0.143>}
 jewelry_cor {parent=hair2_b, loc=<3.7103 -1.2326 0.3337>, rot=<-0.0611 0.7045 -0.0611 -0.7045>}
 MSTR_hair_b {parent=spine_05, loc=<22.3478 -3.5368 0.3337>, rot=<0.2554 -0.9668 0 0>}
 backpack_mover {parent=spine_05, loc=<0 -8.35 0>, rot=<-0.0093 -0.707 0.0093 -0.707>}
 deform_pec_aim_base_l {parent=spine_04, loc=<3.2456 9.3309 -1.2237>, rot=<0.0603 -0.4033 0.3318 -0.8507>}
 deform_pec_aim_l {parent=deform_pec_aim_base_l, loc=<0 0 0>, rot=<0 0 0 1>}
 deform_pec_base_l {parent=deform_pec_aim_l, loc=<9.3665 0 0>, rot=<0 0 0 1>}
 deform_pec_l {parent=deform_pec_base_l, loc=<0 0 0>, rot=<0 0 0 1>}
 deform_pec_aim_base_r {parent=spine_04, loc=<3.2452 9.3309 1.2237>, rot=<-0.4033 -0.0603 0.8507 0.3318>}
 deform_pec_aim_r {parent=deform_pec_aim_base_r, loc=<0 0 0>, rot=<0 0 0 1>}
 deform_pec_base_r {parent=deform_pec_aim_r, loc=<-9.3667 -0.0003 -0.0004>, rot=<0 0 0 1>}
 deform_pec_r {parent=deform_pec_base_r, loc=<0 0 0>, rot=<0 0 0 1>}
 MSTR_skirt1_l {parent=spine_02, loc=<1.9976 6.4205 -8.1404>, rot=<-0.2695 0.5836 0.7189 0.2647>}
 skirt1_l {parent=MSTR_skirt1_l, loc=<0 0 0>, rot=<0 0 0 1>}
 MSTR_skirt2_l {parent=spine_02, loc=<-0.1436 -0.6933 -9.4142>, rot=<-0.2964 0.642 0.6636 0.2443>}
 skirt2_l {parent=MSTR_skirt2_l, loc=<0 0 0>, rot=<0 0 0 1>}
 MSTR_skirt4_l {parent=spine_02, loc=<-3.4569 -7.3723 -2.6254>, rot=<0.4653 -0.8852 0 0>}
 skirt4_l {parent=MSTR_skirt4_l, loc=<0 0 0>, rot=<0 0 0 1>}
 MSTR_skirt3_l {parent=spine_02, loc=<-1.5953 -4.9196 -6.7275>, rot=<0.5151 -0.767 -0.3337 -0.1874>}
 skirt3_l {parent=MSTR_skirt3_l, loc=<0 0 0>, rot=<0 0 0 1>}
 MSTR_skirt1_r {parent=spine_02, loc=<2.1914 6.4355 8.0309>, rot=<0.3342 -0.5823 0.6215 0.4036>}
 skirt1_r {parent=MSTR_skirt1_r, loc=<0 0 0>, rot=<0 0 0 1>}
 MSTR_skirt2_r {parent=spine_02, loc=<-0.2436 -0.7013 9.3151>, rot=<0.2864 -0.6992 0.5901 0.2845>}
 skirt2_r {parent=MSTR_skirt2_r, loc=<0 0 0>, rot=<0 0 0 1>}
 MSTR_skirt3_r {parent=spine_02, loc=<-1.5206 -5.0135 6.6253>, rot=<0.4209 -0.8371 0.1972 0.2884>}
 skirt3_r {parent=MSTR_skirt3_r, loc=<0 0 0>, rot=<0 0 0 1>}
 MSTR_skirt4_r {parent=spine_02, loc=<-3.6338 -7.1204 2.9505>, rot=<0.4653 -0.8852 0 0>}
 skirt4_r {parent=MSTR_skirt4_r, loc=<0 0 0>, rot=<0 0 0 1>}
 thigh_l {parent=pelvis, loc=<-1.0855 -0.2109 -8.7053>, rot=<0.0009 0.0355 0.0155 -0.9993>}
 calf_l {parent=thigh_l, loc=<-42.5284 -0.0019 -0.7297>, rot=<0.0006 -0.0161 -0.0391 -0.9991>}
 foot_l {parent=calf_l, loc=<-46.7401 -2.5014 -0.592>, rot=<0.0427 -0.0238 0.0353 -0.9982>}
 ball_l {parent=foot_l, loc=<-3.5658 10.8066 -0.1544>, rot=<0.0001 -0.0001 -0.7063 -0.7079>}
 calf_twist_02_l {parent=calf_l, loc=<-19.1546 -1.7943 -0.2638>, rot=<0.0237 0.0095 -0.0035 -0.9997>}
 calf_twist_01_l {parent=calf_l, loc=<-35.8483 -1.9197 -0.5794>, rot=<0.0237 0.0095 -0.0035 -0.9997>}
 cor_foot_l {parent=calf_twist_01_l, loc=<-8.41 -4.57 0.828>, rot=<0.6576 -0.7534 0 0>}
 deform_knee_in_base_l {parent=calf_l, loc=<0.8031 -5.0231 0.2456>, rot=<0 0 0 -1>}
 deform_knee_in_l {parent=deform_knee_in_base_l, loc=<0 0 0>, rot=<0 0 0 1>}
 deform_knee_out_base_l {parent=calf_l, loc=<0 4 0>, rot=<0 0 0 -1>}
 deform_knee_out_l {parent=deform_knee_out_base_l, loc=<0 0 0>, rot=<0 0 0 1>}
 thigh_twist_01_l {parent=thigh_l, loc=<-27.952 0 0>, rot=<0.0013 0 0 -1>}
 deform_groin_base_l {parent=thigh_l, loc=<0.4801 5.6776 -0.2486>, rot=<0 0 0 -1>}
 deform_groin_l {parent=deform_groin_base_l, loc=<0 0 0>, rot=<0 0 0 1>}
 cor_skirt_l {parent=thigh_l, loc=<-6.7588 0.0348 -5.6326>, rot=<-0.6816 0.0015 0.7317 0.0019>}
 thigh_r {parent=pelvis, loc=<-1.0855 -0.2109 8.7053>, rot=<0.0355 -0.0009 0.9993 0.0155>}
 calf_r {parent=thigh_r, loc=<42.5284 0.0019 0.7297>, rot=<0.0006 -0.0161 -0.0391 -0.9991>}
 foot_r {parent=calf_r, loc=<46.7401 2.5014 0.592>, rot=<0.0427 -0.0238 0.0353 -0.9982>}
 ball_r {parent=foot_r, loc=<3.5658 -10.8066 0.1544>, rot=<0.0001 -0.0001 -0.7063 -0.7079>}
 calf_twist_02_r {parent=calf_r, loc=<19.1546 1.7943 0.2638>, rot=<0.0237 0.0095 -0.0035 -0.9997>}
 calf_twist_01_r {parent=calf_r, loc=<35.8483 1.9197 0.5794>, rot=<0.0237 0.0095 -0.0035 -0.9997>}
 cor_foot_r {parent=calf_twist_01_r, loc=<8.41 4.57 -0.828>, rot=<0.7504 0.6594 -0.044 -0.011>}
 deform_knee_in_base_r {parent=calf_r, loc=<-0.8031 5.0231 -0.2457>, rot=<0 0 0 -1>}
 deform_knee_in_r {parent=deform_knee_in_base_r, loc=<0 0 0>, rot=<0 0 0 1>}
 deform_knee_out_base_r {parent=calf_r, loc=<0 -4 0>, rot=<0 0 0 -1>}
 deform_knee_out_r {parent=deform_knee_out_base_r, loc=<0 0 0>, rot=<0 0 0 1>}
 thigh_twist_01_r {parent=thigh_r, loc=<27.952 0 0>, rot=<0.0013 0 0 -1>}
 deform_groin_base_r {parent=thigh_r, loc=<-0.4801 -5.6776 0.2486>, rot=<0 0 0 -1>}
 deform_groin_r {parent=deform_groin_base_r, loc=<0 0 0>, rot=<0 0 0 1>}
 cor_skirt_r {parent=thigh_r, loc=<6.7588 -0.0348 5.6325>, rot=<0.0015 0.6816 0.0019 -0.7317>}
 MSTR_skirt_f_l {parent=pelvis, loc=<8.5896 7.656 -4.0789>, rot=<-0.2878 0.4068 0.8586 0.1203>}
 skirt1_f_l {parent=MSTR_skirt_f_l, loc=<0 0 0>, rot=<0 0 0 1>}
 skirt2_f_l {parent=skirt1_f_l, loc=<9.4835 0 0>, rot=<0 0 -0.1007 -0.9949>}
 skirt3_f_l {parent=skirt2_f_l, loc=<11.6618 0 0>, rot=<0 0 0 1>}
 MSTR_skirt_f_r {parent=pelvis, loc=<8.59 7.656 4.079>, rot=<0.282 -0.4108 0.8602 0.1082>}
 skirt1_f_r {parent=MSTR_skirt_f_r, loc=<0 0 0>, rot=<0 0 0 1>}
 skirt2_f_r {parent=skirt1_f_r, loc=<9.48 -0.2666 0>, rot=<0 0 -0.0867 -0.9962>}
 skirt3_f_r {parent=skirt2_f_r, loc=<11.6618 0 0>, rot=<0 0 0 1>}
 MSTR_skirt_b_l {parent=pelvis, loc=<9.2463 -6.8221 -6.1601>, rot=<0.215 -0.9762 -0.0109 -0.0268>}
 skirt1_b_l {parent=MSTR_skirt_b_l, loc=<0 0 0>, rot=<0 0 0 1>}
 skirt2_b_l {parent=skirt1_b_l, loc=<15.6382 0.7561 0.0116>, rot=<-0.0129 -0.0013 -0.1788 -0.9838>}
 skirt3_b_l {parent=skirt2_b_l, loc=<19.1196 0 0>, rot=<-0.0381 0.0195 -0.0837 -0.9956>}
 MSTR_skirt1_b_r {parent=pelvis, loc=<9.2464 -6.8221 6.1601>, rot=<0.204 -0.9788 0.0171 0.0039>}
 skirt1_b_r {parent=MSTR_skirt1_b_r, loc=<0 0 0>, rot=<0 0 0 1>}
 skirt2_b_r {parent=skirt1_b_r, loc=<15.6375 0.4186 -0.6521>, rot=<-0.0065 -0.0134 -0.1679 -0.9857>}
 skirt3_b_r {parent=skirt2_b_r, loc=<19.1196 0 0>, rot=<0 0 0 1>}
 MSTR_skirt_b {parent=pelvis, loc=<7.9993 -8.6038 0>, rot=<0.1601 -0.6887 -0.1601 -0.6887>}
 skirt1_b {parent=MSTR_skirt_b, loc=<0 0 0>, rot=<0 0 0 1>}
 skirt2_b {parent=skirt1_b, loc=<0 0 -8.7321>, rot=<0 0 0 1>}
 MSTR_skirt_d_r {parent=pelvis, loc=<10.051 -1.8795 8.803>, rot=<0.1609 -0.6994 0.6832 0.135>}
 skirt1_d_r {parent=MSTR_skirt_d_r, loc=<0 0 0>, rot=<0 0 0 1>}
 skirt2_d_r {parent=skirt1_d_r, loc=<14.9999 0 0>, rot=<0 0 -0.0762 -0.9971>}
 skirt3_d_r {parent=skirt2_d_r, loc=<20 0 0>, rot=<0 0 0.0076 -1>}
 skirt4_d_r {parent=skirt3_d_r, loc=<20 0 0>, rot=<0 0 0 1>}
 MSTR_skirt_m_r {parent=skirt2_d_r, loc=<0.379 -2.355 0.016>, rot=<0 0 0.0835 -0.9965>}
 skirt1_m_r {parent=MSTR_skirt_m_r, loc=<0 0 0>, rot=<0 0 0 1>}
 MSTR_skirt_d_l {parent=pelvis, loc=<10.0511 -1.8795 -8.803>, rot=<0.1612 -0.6983 -0.6843 -0.1348>}
 skirt1_d_l {parent=MSTR_skirt_d_l, loc=<0 0 0>, rot=<0 0 0 1>}
 skirt2_d_l {parent=skirt1_d_l, loc=<15 0 0>, rot=<0 0 -0.0774 -0.997>}
 skirt3_d_l {parent=skirt2_d_l, loc=<20 0 0>, rot=<0 0 0.0051 -1>}
 skirt4_d_l {parent=skirt3_d_l, loc=<20 0 0>, rot=<0 0 0.0091 -1>}
 MSTR_skirt_m_l {parent=skirt2_d_l, loc=<0.3793 -2.3546 -0.016>, rot=<0 0 0.0142 -0.9999>}
 skirt1_m_l {parent=MSTR_skirt_m_l, loc=<0 0 0>, rot=<0 0 0 1>}
 deform_glute_base_l {parent=pelvis, loc=<-7.3466 -9.0213 -9.1245>, rot=<0.0172 0.031 -0.4477 -0.8935>}
 deform_glute_l {parent=deform_glute_base_l, loc=<0 0 0>, rot=<0 0 0 1>}
 deform_glute_base_r {parent=pelvis, loc=<-7.3464 -9.0213 9.1245>, rot=<0.031 -0.0172 0.8935 -0.4477>}
 deform_glute_r {parent=deform_glute_base_r, loc=<0 0 0>, rot=<0 0 0 1>}
 cor_skirt_b {parent=pelvis, loc=<-16.0018 -13.7442 0>, rot=<0.6943 -0.7197 0 0>}
 attach {parent=root, loc=<0 0 0>, rot=<0 0 0 1>}
 ik_hand_root {parent=root, loc=<0 0 0>, rot=<0 0 0 1>}
 ik_hand_gun {parent=ik_hand_root, loc=<-48.1743 3.3545 104.4381>, rot=<0.6276 0.1309 0.4183 -0.6434>}
 ik_hand_r {parent=ik_hand_gun, loc=<0 0 0>, rot=<0 0 0 1>}
 ik_hand_l {parent=ik_hand_gun, loc=<59.3271 67.6997 34.3524>, rot=<-0.6158 -0.7027 -0.3565 0>}
 ik_foot_root {parent=root, loc=<0 0 0>, rot=<0 0 0 1>}
 ik_foot_l {parent=ik_foot_root, loc=<14.8262 -7.3545 8.296>, rot=<0.0357 0.7045 0.0284 -0.7082>}
 ik_foot_r {parent=ik_foot_root, loc=<-14.8262 -7.3545 8.296>, rot=<0.7082 0.0284 -0.7045 0.0357>}
 root_fp {parent=root, loc=<0 0 0>, rot=<0 0 0 1>}
 attach_fp {parent=root_fp, loc=<0 0 0>, rot=<0 0 0 1>}
 camera_root {parent=root_fp, loc=<0 0 0>, rot=<0 0 0 1>}
 camera {parent=camera_root, loc=<0 0 0>, rot=<0 0 0 1>}
 attach_emote_01 {parent=root, loc=<0 0 0>, rot=<0 0 0 1>}
 attach_emote_02 {parent=root, loc=<0 0 0>, rot=<0 0 0 1>}
 attach_emote_03 {parent=root, loc=<0 0 0>, rot=<0 0 0 1>}
 warp_point_01 {parent=root, loc=<0 0 0>, rot=<0 0 0 1>}
 warp_point_02 {parent=root, loc=<0 0 0>, rot=<0 0 0 1>}
 warp_point_03 {parent=root, loc=<0 0 0>, rot=<0 0 0 1>}
]
