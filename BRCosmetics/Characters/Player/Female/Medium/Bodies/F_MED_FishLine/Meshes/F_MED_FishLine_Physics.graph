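# Physics asset for the F_MED_FishLine female player body cosmetic, providing jointed skirt bodies plus pelvis and leg capsules.
# Six skirt chains (front, b and d, left and right) hang from small MSTR anchors with tight angular limits.
# Anchors and legs lock angularly to their parents, except the angularly free MSTR_skirt_d_r.
# Angularly free cross-links join some right-side chains and front skirts to thighs; a huge root sphere forms a near-flat floor.
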PhysicsAsset /BRCosmetics/Characters/Player/Female/Medium/Bodies/F_MED_FishLine/Meshes/F_MED_FishLine_Physics
bodies {
 MSTR_skirt1_b_r [{shapeType=Capsule, length=1, radius=1}]
 MSTR_skirt_b_l [{shapeType=Capsule, length=0.505, radius=0.505, shapeName=MSTR_skirt_b_l_capsule, rotation=<rot 0 -90 0>}]
 MSTR_skirt_d_l [{shapeType=Capsule, length=0.505, radius=0.505, shapeName=MSTR_skirt_d_l_capsule, rotation=<rot 0 -90 180>}]
 MSTR_skirt_d_r [{shapeType=Capsule, length=0.505, radius=0.505, shapeName=MSTR_skirt_d_r_capsule, rotation=<rot 0 -90 0>}]
 MSTR_skirt_f_l [{shapeType=Capsule, length=0.505, radius=0.505, shapeName=MSTR_skirt_f_l_capsule}]
 MSTR_skirt_f_r [{shapeType=Capsule, length=0.505, radius=0.505, shapeName=MSTR_skirt_f_r_capsule}]
 calf_l [{shapeType=Capsule, length=36, radius=7, center=<-19.2913 -3.4151 0.0171>, rotation=<rot 90 179.9996 0>}]
 calf_r [{shapeType=Capsule, length=51.2667, radius=7, center=<19.2913 3.4151 -0.0171>, rotation=<rot 90 -0.0004 0>}]
 foot_l [{shapeType=Capsule, length=13.379, radius=8.128, center=<-1.4567 6.1773 0.0506>, rotation=<rot 0.5225 -0.7831 92.0866>}]
 foot_r [{shapeType=Capsule, length=13.379, radius=8.128, center=<1.4567 -6.1773 -0.0506>, rotation=<rot 0.5225 -0.7831 -88.3591>}]
 pelvis [
  {shapeType=Capsule, length=8, radius=11.8236, center=<-1.257 -1.9323 0>}
  {shapeType=Capsule, length=16.5, radius=7.5, shapeName=pelvis_capsule, center=<-9.4595 -3.2438 1.294>}
 ]
 root [{shapeType=Sphere, radius=500, shapeName=root_sphere, center=<0 0 -495>}]
 skirt1_b_l [{shapeType=Capsule, length=10, radius=7, center=<8.6677 -8.2 -3.3331>, rotation=<rot 0 0 -180>}]
 skirt1_b_r [{shapeType=Capsule, length=10, radius=7, center=<8.6677 -8.2 3.3331>, rotation=<rot 0 0.0002 0>}]
 skirt1_d_l [{shapeType=Capsule, length=10, radius=6, shapeName=skirt1_d_l_capsule, center=<8.5 -8.46 0>, rotation=<rot 0 -90 180>}]
 skirt1_d_r [{shapeType=Capsule, length=0.505, radius=8.5, shapeName=skirt1_d_r_capsule, center=<13.4164 -9.6141 0>, rotation=<rot 0 0 -0.4457>}]
 skirt1_f_l [{shapeType=Capsule, length=6, radius=4.5, shapeName=skirt1_f_l_capsule, center=<5.1067 -2.6569 0.6904>, rotation=<rot -90 -0.0006 0>}]
 skirt1_f_r [{shapeType=Capsule, length=6, radius=4.5, shapeName=skirt1_f_r_capsule, center=<5.2877 -3.4458 0>, rotation=<rot 90 -2.2866 0>}]
 skirt2_b_l [{shapeType=Capsule, length=10, radius=7, center=<1.8215 -6.2077 -0.7715>, rotation=<rot 0 10.0006 5>}]
 skirt2_b_r [{shapeType=Capsule, length=18.3647, radius=7, center=<2.8549 -5.0238 -6.3136>, rotation=<rot 0 10.0007 169.5543>}]
 skirt2_d_l [{shapeType=Capsule, length=20, radius=10, shapeName=skirt2_d_l_capsule, center=<4.3245 -10.3968 0.662>, rotation=<rot 0.0004 0 1.4423>}]
 skirt2_d_r [{
   shapeType=Capsule
   length=13.8663
   radius=8.5
   shapeName=skirt2_d_r_capsule
   center=<11.7369 -8.4374 -0.2763>
   rotation=<rot -0.0003 0 14.5544>
  }]
 skirt2_f_l [{shapeType=Capsule, length=8, radius=4.5, shapeName=skirt2_f_l_capsule, center=<5.2071 -2.2503 0>, rotation=<rot -90 6.0425 0>}]
 skirt2_f_r [{shapeType=Capsule, length=8, radius=4.5, shapeName=skirt2_f_r_capsule, center=<4.9329 -4.0714 0.8425>, rotation=<rot 90 0.8261 0>}]
 skirt3_b_l [{shapeType=Capsule, length=19.5231, radius=7, center=<1.3906 0.4147 3.8982>, rotation=<rot -52.7886 -44.0968 -139.3962>}]
 skirt3_b_r [{shapeType=Capsule, length=31.0929, radius=7, center=<1.3924 -0.1194 -3.7625>, rotation=<rot 52.3463 0.0219 -8.4364>}]
 skirt3_d_l [{shapeType=Capsule, length=10, radius=10, center=<2 -8.3 -0.2>, rotation=<rot 0 0 0.3647>}]
 skirt3_d_r [{shapeType=Capsule, length=10.8737, radius=8.5, shapeName=skirt3_d_r_capsule, center=<4.9328 -5.1494 3.4047>, rotation=<rot -0.0001 0 30>}]
 skirt3_f_l [{
   shapeType=Capsule
   length=16.0655
   radius=4.5
   shapeName=skirt3_f_l_capsule
   center=<5.1361 -0.2543 -0.4938>
   rotation=<rot 0.2822 102.3341 -91.3981>
  }]
 skirt3_f_r [{shapeType=Capsule, length=8, radius=5, shapeName=skirt3_f_r_capsule, center=<6.2 -0.7789 1.2137>, rotation=<rot 86.3964 -96.5536 -113.0393>}]
 skirt4_d_l [{
   shapeType=Capsule
   length=10
   radius=8.5
   shapeName=skirt4_d_l_capsule
   center=<5.9463 -3.5486 0.427>
   rotation=<rot 4.8921 64.2636 -68.5297>
  }]
 skirt4_d_r [{
   shapeType=Capsule
   length=21.4873
   radius=9.4761
   shapeName=skirt4_d_r_capsule
   center=<1.1929 -2.0601 1.1493>
   rotation=<rot -1.6454 -106.2613 -71.1729>
  }]
 thigh_l [
  {shapeType=Capsule, length=36, radius=9.635, center=<-15.5115 -0.7639 0.6997>, rotation=<rot -85.8495 1.7197 1.0817>}
  {shapeType=Capsule, length=19.5, radius=9, shapeName=thigh_l_capsule, center=<-13.1391 0.0001 0.5019>, rotation=<rot 85 179.9997 179.9999>}
 ]
 thigh_r [
  {shapeType=Capsule, length=38.4432, radius=9.816, center=<20.9427 2.2658 -0.3951>, rotation=<rot -86.8174 1.9076 177.0418>}
  {shapeType=Capsule, length=19.5, radius=9, shapeName=thigh_r_capsule, center=<13.3243 0.7112 -1.1671>, rotation=<rot 79.3782 178.5036 1.3292>}
 ]
}
constraints [
 {
  bone1Name=thigh_l
  bone2Name=pelvis
  swing1LimitDegrees=45
  swing1Motion=Locked
  swing2LimitDegrees=45
  swing2Motion=Locked
  twistLimitDegrees=45
  twistMotion=Locked
 }
 {
  bone1Name=calf_l
  bone2Name=thigh_l
  swing1LimitDegrees=45
  swing1Motion=Locked
  swing2LimitDegrees=45
  swing2Motion=Locked
  twistLimitDegrees=45
  twistMotion=Locked
 }
 {
  bone1Name=foot_l
  bone2Name=calf_l
  swing1LimitDegrees=45
  swing1Motion=Locked
  swing2LimitDegrees=45
  swing2Motion=Locked
  twistLimitDegrees=45
  twistMotion=Locked
 }
 {
  bone1Name=thigh_r
  bone2Name=pelvis
  swing1LimitDegrees=45
  swing1Motion=Locked
  swing2LimitDegrees=45
  swing2Motion=Locked
  twistLimitDegrees=45
  twistMotion=Locked
 }
 {
  bone1Name=calf_r
  bone2Name=thigh_r
  swing1LimitDegrees=45
  swing1Motion=Locked
  swing2LimitDegrees=45
  swing2Motion=Locked
  twistLimitDegrees=45
  twistMotion=Locked
 }
 {
  bone1Name=foot_r
  bone2Name=calf_r
  swing1LimitDegrees=45
  swing1Motion=Locked
  swing2LimitDegrees=45
  swing2Motion=Locked
  twistLimitDegrees=45
  twistMotion=Locked
 }
 {bone1Name=skirt2_b_l, bone2Name=skirt1_b_l, swing1LimitDegrees=23, swing2LimitDegrees=10, twistLimitDegrees=1}
 {bone1Name=skirt2_b_r, bone2Name=skirt1_b_r, swing1LimitDegrees=23, swing2LimitDegrees=10, twistLimitDegrees=1}
 {bone1Name=skirt3_b_r, bone2Name=skirt2_b_r, swing1LimitDegrees=20, swing2LimitDegrees=10, twistLimitDegrees=1}
 {bone1Name=skirt1_b_r, bone2Name=MSTR_skirt1_b_r, swing1LimitDegrees=20, swing2LimitDegrees=4, twistLimitDegrees=1}
 {
  bone1Name=MSTR_skirt1_b_r
  bone2Name=pelvis
  swing1LimitDegrees=45
  swing1Motion=Locked
  swing2LimitDegrees=45
  swing2Motion=Locked
  twistLimitDegrees=45
  twistMotion=Locked
 }
 {bone1Name=skirt1_b_l, bone2Name=MSTR_skirt_b_l, swing1LimitDegrees=20, swing2LimitDegrees=4, twistLimitDegrees=0}
 {bone1Name=skirt3_b_l, bone2Name=skirt2_b_l, swing1LimitDegrees=20, swing2LimitDegrees=10, twistLimitDegrees=0}
 {
  bone1Name=MSTR_skirt_b_l
  bone2Name=pelvis
  swing1LimitDegrees=45
  swing1Motion=Locked
  swing2LimitDegrees=45
  swing2Motion=Locked
  twistLimitDegrees=45
  twistMotion=Locked
 }
 {
  bone1Name=MSTR_skirt_d_l
  bone2Name=pelvis
  swing1LimitDegrees=45
  swing1Motion=Locked
  swing2LimitDegrees=45
  swing2Motion=Locked
  twistLimitDegrees=45
  twistMotion=Locked
 }
 {bone1Name=skirt1_d_l, bone2Name=MSTR_skirt_d_l, swing1LimitDegrees=10, swing2LimitDegrees=10, twistLimitDegrees=1}
 {bone1Name=skirt2_d_l, bone2Name=skirt1_d_l, swing1LimitDegrees=20, swing2LimitDegrees=10, twistLimitDegrees=2}
 {bone1Name=skirt3_d_l, bone2Name=skirt2_d_l, swing1LimitDegrees=15, swing2LimitDegrees=15, twistLimitDegrees=2}
 {bone1Name=skirt4_d_l, bone2Name=skirt3_d_l, swing1LimitDegrees=15, swing2LimitDegrees=15, twistLimitDegrees=1}
 {
  bone1Name=pelvis
  bone2Name=root
  swing1LimitDegrees=45
  swing1Motion=Locked
  swing2LimitDegrees=45
  swing2Motion=Locked
  twistLimitDegrees=45
  twistMotion=Locked
 }
 {
  bone1Name=MSTR_skirt_f_l
  bone2Name=pelvis
  swing1LimitDegrees=45
  swing1Motion=Locked
  swing2LimitDegrees=45
  swing2Motion=Locked
  twistLimitDegrees=45
  twistMotion=Locked
 }
 {bone1Name=skirt1_f_l, bone2Name=MSTR_skirt_f_l, swing1LimitDegrees=10, swing2LimitDegrees=10, twistLimitDegrees=0}
 {bone1Name=skirt2_f_l, bone2Name=skirt1_f_l, swing1LimitDegrees=15, swing2LimitDegrees=15, twistLimitDegrees=0}
 {
  bone1Name=MSTR_skirt_f_r
  bone2Name=pelvis
  swing1LimitDegrees=45
  swing1Motion=Locked
  swing2LimitDegrees=45
  swing2Motion=Locked
  twistLimitDegrees=45
  twistMotion=Locked
 }
 {bone1Name=skirt1_f_r, bone2Name=MSTR_skirt_f_r, swing1LimitDegrees=15, swing2LimitDegrees=8, twistLimitDegrees=0, twistMotion=Locked}
 {bone1Name=skirt2_f_r, bone2Name=skirt1_f_r, swing1LimitDegrees=15, swing2LimitDegrees=10, twistLimitDegrees=0}
 {bone1Name=skirt3_f_r, bone2Name=skirt2_f_r, swing1LimitDegrees=15, swing2LimitDegrees=10, twistLimitDegrees=0}
 {
  bone1Name=MSTR_skirt_d_r
  bone2Name=pelvis
  swing1LimitDegrees=45
  swing1Motion=Free
  swing2LimitDegrees=45
  swing2Motion=Free
  twistLimitDegrees=45
  twistMotion=Free
 }
 {bone1Name=skirt1_d_r, bone2Name=MSTR_skirt_d_r, swing1LimitDegrees=25, swing2LimitDegrees=10, twistLimitDegrees=0}
 {bone1Name=skirt2_d_r, bone2Name=skirt1_d_r, swing1LimitDegrees=15, swing2LimitDegrees=10, twistLimitDegrees=0}
 {bone1Name=skirt3_d_r, bone2Name=skirt2_d_r, swing1LimitDegrees=15, swing2LimitDegrees=15, twistLimitDegrees=2}
 {bone1Name=skirt4_d_r, bone2Name=skirt3_d_r, swing1LimitDegrees=15, swing2LimitDegrees=15, twistLimitDegrees=1}
 {
  bone1Name=skirt2_d_r
  bone2Name=skirt3_f_r
  swing1LimitDegrees=45
  swing1Motion=Free
  swing2LimitDegrees=45
  swing2Motion=Free
  twistLimitDegrees=45
  twistMotion=Free
 }
 {
  bone1Name=skirt2_b_r
  bone2Name=skirt2_d_r
  swing1LimitDegrees=45
  swing1Motion=Free
  swing2LimitDegrees=45
  swing2Motion=Free
  twistLimitDegrees=45
  twistMotion=Free
 }
 {bone1Name=skirt3_f_l, bone2Name=skirt2_f_l, swing1LimitDegrees=25, swing2LimitDegrees=10, twistLimitDegrees=0}
 {
  bone1Name=skirt3_f_l
  bone2Name=thigh_l
  swing1LimitDegrees=45
  swing1Motion=Free
  swing2LimitDegrees=45
  swing2Motion=Free
  twistLimitDegrees=45
  twistMotion=Free
 }
 {
  bone1Name=skirt3_f_r
  bone2Name=thigh_r
  swing1LimitDegrees=45
  swing1Motion=Free
  swing2LimitDegrees=45
  swing2Motion=Free
  twistLimitDegrees=45
  twistMotion=Free
 }
]
