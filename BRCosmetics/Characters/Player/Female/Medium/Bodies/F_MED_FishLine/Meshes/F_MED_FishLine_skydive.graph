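# Physics asset for the F_MED_FishLine female body's skydive mesh, defining capsule bodies and constraints for skirt motion.
# Six skirt chains (front, back and "d" on each side) hang from MSTR anchors on the pelvis with tight angular limits.
# Pelvis and leg joints are angle-locked, while MSTR_skirt_d_r and cross-links between skirt chains and to thighs are angularly free.
# A huge sphere sits below the root, its top near root height.
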PhysicsAsset /BRCosmetics/Characters/Player/Female/Medium/Bodies/F_MED_FishLine/Meshes/F_MED_FishLine_skydive
bodies {
 MSTR_skirt1_b_r [{shapeType=Capsule, length=1, radius=1, center=<-0.0306 0 0>}]
 MSTR_skirt_b_l [{shapeType=Capsule, length=0.505, radius=0.505, shapeName=MSTR_skirt_b_l_capsule, rotation=<rot 0 -90 0>}]
 MSTR_skirt_d_l [{shapeType=Capsule, length=0.505, radius=0.505, shapeName=MSTR_skirt_d_l_capsule, rotation=<rot 0 -90 180>}]
 MSTR_skirt_d_r [{shapeType=Capsule, length=0.505, radius=0.505, shapeName=MSTR_skirt_d_r_capsule, rotation=<rot 0 -90 0>}]
 MSTR_skirt_f_l [{shapeType=Capsule, length=0.505, radius=0.505, shapeName=MSTR_skirt_f_l_capsule}]
 MSTR_skirt_f_r [{shapeType=Capsule, length=0.505, radius=0.505, shapeName=MSTR_skirt_f_r_capsule}]
 calf_l [{shapeType=Capsule, length=36, radius=7, center=<-19.2913 -3.4151 0.0171>, rotation=<rot 90 179.9996 0>}]
 calf_r [{shapeType=Capsule, length=36, radius=7, center=<19.2913 3.4151 -0.0171>, rotation=<rot 90 -0.0004 0>}]
 foot_l [{shapeType=Capsule, length=13.379, radius=8.128, center=<-1.4873 6.1777 0.0503>, rotation=<rot 0.5225 -0.7831 92.0866>}]
 foot_r [{shapeType=Capsule, length=13.379, radius=8.128, center=<1.4567 -6.1773 -0.0506>, rotation=<rot 0.5225 -0.7831 -87.9134>}]
 pelvis [
  {shapeType=Capsule, length=8, radius=10.5, center=<0 -1.4752 0>, rotation=<rot 0 -0.0004 0>}
  {shapeType=Capsule, length=13.5, radius=7.5, shapeName=pelvis_capsule, center=<-10.2433 -5.3277 0>}
 ]
 root [{shapeType=Sphere, radius=500, shapeName=root_sphere, center=<0 0 -495>}]
 skirt1_b_l [{shapeType=Capsule, length=10, radius=7, center=<8.6677 -8.2 -3.3331>, rotation=<rot 0 0 -180>}]
 skirt1_b_r [{shapeType=Capsule, length=10, radius=7, center=<8.6677 -8.2 3.3331>, rotation=<rot 0 0.0002 0>}]
 skirt1_d_l [{shapeType=Capsule, length=10, radius=6, shapeName=skirt1_d_l_capsule, center=<8.5 -8.46 0>, rotation=<rot 0 -90 180>}]
 skirt1_d_r [{shapeType=Capsule, length=0.505, radius=8.5, shapeName=skirt1_d_r_capsule, center=<13.4164 -9.6141 0>}]
 skirt1_f_l [{shapeType=Capsule, length=6, radius=4.5, shapeName=skirt1_f_l_capsule, center=<5.1067 -2.6569 0.6904>, rotation=<rot -90 -0.0006 0>}]
 skirt1_f_r [{shapeType=Capsule, length=6, radius=4.5, shapeName=skirt1_f_r_capsule, center=<5.2877 -3.4458 0>, rotation=<rot 90 -2.2866 0>}]
 skirt2_b_l [{shapeType=Capsule, length=15, radius=7, center=<1.5188 -4.4912 -0.924>}]
 skirt2_b_r [{shapeType=Capsule, length=15, radius=7, center=<1.5732 -3.8277 -0.7887>}]
 skirt2_d_l [{shapeType=Capsule, length=20, radius=10, shapeName=skirt2_d_l_capsule, center=<5 -10.1525 0.7075>, rotation=<rot 0.0003 0 -2.0337>}]
 skirt2_d_r [{
   shapeType=Capsule
   length=13.8663
   radius=8.5
   shapeName=skirt2_d_r_capsule
   center=<11.7369 -8.4374 -0.2763>
   rotation=<rot -0.0003 0 15.0001>
  }]
 skirt2_f_l [{shapeType=Capsule, length=6, radius=4.5, shapeName=skirt2_f_l_capsule, center=<4.919 -2.2808 0>, rotation=<rot -90 6.0425 0>}]
 skirt2_f_r [{shapeType=Capsule, length=6, radius=4.5, shapeName=skirt2_f_r_capsule, center=<4.895 -2.0999 0.8425>, rotation=<rot 90 -1.9861 0>}]
 skirt3_b_l [{shapeType=Capsule, length=12, radius=7, center=<1.4915 1 3.6439>, rotation=<rot -52.7886 -44.0971 -147.8337>}]
 skirt3_b_r [{shapeType=Capsule, length=12, radius=7, center=<1.2605 1 -3.6611>, rotation=<rot 52.3463 0.0217 0.0011>}]
 skirt3_d_l [{shapeType=Capsule, length=10, radius=10, center=<2 -8.3 -0.2>, rotation=<rot 0 0 -10>}]
 skirt3_d_r [{shapeType=Capsule, length=10.8737, radius=8.5, shapeName=skirt3_d_r_capsule, center=<4.9328 -5.1494 3.4047>, rotation=<rot -0.0001 0 30>}]
 skirt3_f_l [{
   shapeType=Capsule
   length=7.0043
   radius=4.5
   shapeName=skirt3_f_l_capsule
   center=<2.3584 -1.7773 -1.3424>
   rotation=<rot 0.0039 91.0877 -91.4265>
  }]
 skirt3_f_r [{
   shapeType=Capsule
   length=6
   radius=6
   shapeName=skirt3_f_r_capsule
   center=<3.6473 -2.6224 2.9864>
   rotation=<rot 79.6123 -145.0482 -152.8516>
  }]
 skirt4_d_l [{
   shapeType=Capsule
   length=10
   radius=8.5
   shapeName=skirt4_d_l_capsule
   center=<5.488 -4.4994 0.3367>
   rotation=<rot 4.8921 64.2636 -61.3713>
  }]
 skirt4_d_r [{
   shapeType=Capsule
   length=10
   radius=8.5
   shapeName=skirt4_d_r_capsule
   center=<7.9167 -4.0564 2.3195>
   rotation=<rot -1.6455 -106.2613 -70.8298>
  }]
 thigh_l [
  {shapeType=Capsule, length=36, radius=8.5, center=<-15.5 -1 0.7>, rotation=<rot -85 0 0>}
  {shapeType=Capsule, length=19.5, radius=9, shapeName=thigh_l_capsule, center=<-13.1391 0.0001 0.5019>, rotation=<rot 85 179.9997 179.9999>}
 ]
 thigh_r [
  {shapeType=Capsule, length=36, radius=8.5, center=<15.5 1 -0.7>, rotation=<rot -85 0 180>}
  {shapeType=Capsule, length=19.5, radius=9, shapeName=thigh_r_capsule, center=<13.1391 -0.0001 -0.5019>, rotation=<rot 85 179.9997 -0.0001>}
 ]
}
constraints [
 {
  bone1Name=thigh_l
  bone2Name=pelvis
  swing1LimitDegrees=45
  swing1Motion=Locked
  swing2LimitDegrees=45
  swing2Motion=Locked
  twistLimitDegrees=45
  twistMotion=Locked
 }
 {
  bone1Name=calf_l
  bone2Name=thigh_l
  swing1LimitDegrees=45
  swing1Motion=Locked
  swing2LimitDegrees=45
  swing2Motion=Locked
  twistLimitDegrees=45
  twistMotion=Locked
 }
 {
  bone1Name=foot_l
  bone2Name=calf_l
  swing1LimitDegrees=45
  swing1Motion=Locked
  swing2LimitDegrees=45
  swing2Motion=Locked
  twistLimitDegrees=45
  twistMotion=Locked
 }
 {
  bone1Name=thigh_r
  bone2Name=pelvis
  swing1LimitDegrees=45
  swing1Motion=Locked
  swing2LimitDegrees=45
  swing2Motion=Locked
  twistLimitDegrees=45
  twistMotion=Locked
 }
 {
  bone1Name=calf_r
  bone2Name=thigh_r
  swing1LimitDegrees=45
  swing1Motion=Locked
  swing2LimitDegrees=45
  swing2Motion=Locked
  twistLimitDegrees=45
  twistMotion=Locked
 }
 {
  bone1Name=foot_r
  bone2Name=calf_r
  swing1LimitDegrees=45
  swing1Motion=Locked
  swing2LimitDegrees=45
  swing2Motion=Locked
  twistLimitDegrees=45
  twistMotion=Locked
 }
 {bone1Name=skirt2_b_l, bone2Name=skirt1_b_l, swing1LimitDegrees=35, swing2LimitDegrees=10, twistLimitDegrees=1}
 {bone1Name=skirt2_b_r, bone2Name=skirt1_b_r, swing1LimitDegrees=35, swing2LimitDegrees=10, twistLimitDegrees=1}
 {bone1Name=skirt3_b_r, bone2Name=skirt2_b_r, swing1LimitDegrees=20, swing2LimitDegrees=10, twistLimitDegrees=1}
 {bone1Name=skirt1_b_r, bone2Name=MSTR_skirt1_b_r, swing1LimitDegrees=10, swing2LimitDegrees=4, twistLimitDegrees=1}
 {
  bone1Name=MSTR_skirt1_b_r
  bone2Name=pelvis
  swing1LimitDegrees=45
  swing1Motion=Locked
  swing2LimitDegrees=45
  swing2Motion=Locked
  twistLimitDegrees=45
  twistMotion=Locked
 }
 {bone1Name=skirt1_b_l, bone2Name=MSTR_skirt_b_l, swing1LimitDegrees=10, swing2LimitDegrees=4, twistLimitDegrees=0}
 {bone1Name=skirt3_b_l, bone2Name=skirt2_b_l, swing1LimitDegrees=20, swing2LimitDegrees=10, twistLimitDegrees=0}
 {
  bone1Name=MSTR_skirt_b_l
  bone2Name=pelvis
  swing1LimitDegrees=45
  swing1Motion=Locked
  swing2LimitDegrees=45
  swing2Motion=Locked
  twistLimitDegrees=45
  twistMotion=Locked
 }
 {
  bone1Name=MSTR_skirt_d_l
  bone2Name=pelvis
  swing1LimitDegrees=45
  swing1Motion=Locked
  swing2LimitDegrees=45
  swing2Motion=Locked
  twistLimitDegrees=45
  twistMotion=Locked
 }
 {bone1Name=skirt1_d_l, bone2Name=MSTR_skirt_d_l, swing1LimitDegrees=10, swing2LimitDegrees=10, twistLimitDegrees=1}
 {bone1Name=skirt2_d_l, bone2Name=skirt1_d_l, swing1LimitDegrees=20, swing2LimitDegrees=10, twistLimitDegrees=2}
 {bone1Name=skirt3_d_l, bone2Name=skirt2_d_l, swing1LimitDegrees=10, swing2LimitDegrees=10, twistLimitDegrees=2}
 {bone1Name=skirt4_d_l, bone2Name=skirt3_d_l, swing1LimitDegrees=10, swing2LimitDegrees=10, twistLimitDegrees=0}
 {
  bone1Name=pelvis
  bone2Name=root
  swing1LimitDegrees=45
  swing1Motion=Locked
  swing2LimitDegrees=45
  swing2Motion=Locked
  twistLimitDegrees=45
  twistMotion=Locked
 }
 {
  bone1Name=MSTR_skirt_f_l
  bone2Name=pelvis
  swing1LimitDegrees=45
  swing1Motion=Locked
  swing2LimitDegrees=45
  swing2Motion=Locked
  twistLimitDegrees=45
  twistMotion=Locked
 }
 {bone1Name=skirt1_f_l, bone2Name=MSTR_skirt_f_l, swing1LimitDegrees=10, swing2LimitDegrees=10, twistLimitDegrees=0}
 {bone1Name=skirt2_f_l, bone2Name=skirt1_f_l, swing1LimitDegrees=15, swing2LimitDegrees=15, twistLimitDegrees=0}
 {
  bone1Name=MSTR_skirt_f_r
  bone2Name=pelvis
  swing1LimitDegrees=45
  swing1Motion=Locked
  swing2LimitDegrees=45
  swing2Motion=Locked
  twistLimitDegrees=45
  twistMotion=Locked
 }
 {bone1Name=skirt1_f_r, bone2Name=MSTR_skirt_f_r, swing1LimitDegrees=15, swing2LimitDegrees=8, twistLimitDegrees=0, twistMotion=Locked}
 {bone1Name=skirt2_f_r, bone2Name=skirt1_f_r, swing1LimitDegrees=15, swing2LimitDegrees=10, twistLimitDegrees=0}
 {bone1Name=skirt3_f_r, bone2Name=skirt2_f_r, swing1LimitDegrees=15, swing2LimitDegrees=10, twistLimitDegrees=0}
 {
  bone1Name=MSTR_skirt_d_r
  bone2Name=pelvis
  swing1LimitDegrees=45
  swing1Motion=Free
  swing2LimitDegrees=45
  swing2Motion=Free
  twistLimitDegrees=45
  twistMotion=Free
 }
 {bone1Name=skirt1_d_r, bone2Name=MSTR_skirt_d_r, swing1LimitDegrees=25, swing2LimitDegrees=10, twistLimitDegrees=0}
 {bone1Name=skirt2_d_r, bone2Name=skirt1_d_r, swing1LimitDegrees=15, swing2LimitDegrees=10, twistLimitDegrees=0}
 {bone1Name=skirt3_d_r, bone2Name=skirt2_d_r, swing1LimitDegrees=10, swing2LimitDegrees=10, twistLimitDegrees=0}
 {bone1Name=skirt4_d_r, bone2Name=skirt3_d_r, swing1LimitDegrees=10, swing2LimitDegrees=10, twistLimitDegrees=0}
 {
  bone1Name=skirt2_d_r
  bone2Name=skirt3_f_r
  swing1LimitDegrees=45
  swing1Motion=Free
  swing2LimitDegrees=45
  swing2Motion=Free
  twistLimitDegrees=45
  twistMotion=Free
 }
 {
  bone1Name=skirt2_b_r
  bone2Name=skirt2_d_r
  swing1LimitDegrees=45
  swing1Motion=Free
  swing2LimitDegrees=45
  swing2Motion=Free
  twistLimitDegrees=45
  twistMotion=Free
 }
 {bone1Name=skirt3_f_l, bone2Name=skirt2_f_l, swing1LimitDegrees=25, swing2LimitDegrees=10, twistLimitDegrees=0}
 {
  bone1Name=skirt3_f_l
  bone2Name=thigh_l
  swing1LimitDegrees=45
  swing1Motion=Free
  swing2LimitDegrees=45
  swing2Motion=Free
  twistLimitDegrees=45
  twistMotion=Free
 }
 {
  bone1Name=skirt3_f_r
  bone2Name=thigh_r
  swing1LimitDegrees=45
  swing1Motion=Free
  swing2LimitDegrees=45
  swing2Motion=Free
  twistLimitDegrees=45
  twistMotion=Free
 }
 {
  bone1Name=skirt3_b_r
  bone2Name=skirt3_d_r
  swing1LimitDegrees=45
  swing1Motion=Free
  swing2LimitDegrees=45
  swing2Motion=Free
  twistLimitDegrees=45
  twistMotion=Free
 }
 {
  bone1Name=skirt3_b_l
  bone2Name=skirt3_d_l
  swing1LimitDegrees=45
  swing1Motion=Free
  swing2LimitDegrees=45
  swing2Motion=Free
  twistLimitDegrees=45
  twistMotion=Free
 }
 {
  bone1Name=skirt2_b_l
  bone2Name=skirt2_d_l
  swing1LimitDegrees=45
  swing1Motion=Free
  swing2LimitDegrees=45
  swing2Motion=Free
  twistLimitDegrees=45
  twistMotion=Free
 }
 {
  bone1Name=skirt2_b_r
  bone2Name=skirt2_b_l
  swing1LimitDegrees=45
  swing1Motion=Free
  swing2LimitDegrees=45
  swing2Motion=Free
  twistLimitDegrees=45
  twistMotion=Free
 }
 {
  bone1Name=skirt3_b_l
  bone2Name=skirt3_b_r
  swing1LimitDegrees=45
  swing1Motion=Free
  swing2LimitDegrees=45
  swing2Motion=Free
  twistLimitDegrees=45
  twistMotion=Free
 }
]
