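# Post-process style cosmetic AnimBP for the FishLine female body, adding secondary motion to skirt, shoulder and ring bones.
# It copies the parent mesh pose, applies weighted bone constraints, offsets and a Control Rig, then caches it.
# Two RigidBody sims read that cache, with skydiving or parachute state selecting the skydive physics asset, before chained AnimDynamics finish the pose.
# The event graph computes GravityOverride, but the AnimGraph uses fixed gravity values.
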AnimBlueprint /BRCosmetics/Characters/Player/Female/Medium/Bodies/F_MED_FishLine/Meshes/F_MED_FishLine_AnimBP
target_skeleton=&/F_MED_FishLine_Skeleton
graphs [
 AnimGraph {nodes [
   Anim Dynamics - Bone: skirt1_b(Alpha=1, ExternalForce="0,0,0") {
    bUseGravityOverride=true
    simulationSpace=RootRelative
    boundBone {boneName=skirt1_b}
    componentLinearAccScale=<1 1 1>
    componentLinearVelScale=<1 1 1>
    gravityOverride {z=-980}
    physicsBodyDefinitions [{
      boundBone {boneName=skirt1_b}
      boxExtents=<10 10 10>
      constraintSetup {angularTargetAxis=AxisZ, angularLimitsMax=<26.7769 0 0>, angularLimitsMin=<-71.4046 0 0>, angularTarget=<1 0 1>}
      localJointOffset=<0 0 1>
     }]
   }
   n1 = RigidBody!(ComponentPose=Local To Component(LocalPose=Use cached pose 'cech ctrl'()), Alpha, OverrideWorldGravity:Vector="0,0,-980", ExternalForce) -> (Pose) {
    bOverrideWorldGravity=true
    overridePhysicsAsset=&/F_MED_FishLine_Physics
    simulationSpace=BaseBoneSpace
    baseBoneRef {boneName=pelvis}
    overrideWorldGravity {z=-980}
    simSpaceSettings {
     dampingAlpha=0
     maxAngularAcceleration=75.5535
     maxAngularVelocity=8
     maxLinearAcceleration=1800
     maxLinearVelocity=550
     velocityScaleZ=0.648
     worldAlpha=1
    }
   }
   n2 = Component To Local(ComponentPose=@n1)
   n3 = RigidBody(ComponentPose=Local To Component(LocalPose=Use cached pose 'cech ctrl'()), OverrideWorldGravity="0,0,-980") {
    bOverrideWorldGravity=true
    overridePhysicsAsset=&/F_MED_FishLine_skydive
    simulationSpace=BaseBoneSpace
    baseBoneRef {boneName=pelvis}
    overrideWorldGravity {z=-980}
    simSpaceSettings {dampingAlpha=0, maxAngularAcceleration=32, maxAngularVelocity=8, maxLinearAcceleration=2550, maxLinearVelocity=700, worldAlpha=1}
   }
   n4 = Component To Local(ComponentPose=@n3)
   n5 = Get isfullskydivie() -> (isfullskydivie:bool)
   n6 = Anim Dynamics - Bone: skirt1_m_r!(
    ComponentPose=Local To Component(LocalPose=Blend Poses by bool(BlendPose_0=@n4, BlendPose_1=@n2, bActiveValue=@n5))
    Alpha=1
    GravityOverride:Vector="0,0,-900"
    ExternalForce="0,0,0"
   ) -> (Pose) {
    angularDampingOverride=1
    angularSpringConstant=950
    bAngularSpring=true
    bOverrideAngularDamping=true
    bUseGravityOverride=true
    simulationSpace=RootRelative
    alphaBoolBlend {blendOption=Linear}
    boundBone {boneName=skirt1_m_r}
    componentAppliedLinearAccClamp=<4000 4000 4000>
    componentLinearAccScale=<0.4 0.4 0.42>
    componentLinearVelScale=<0.4 0.4 0.1>
    gravityOverride {z=-900}
    physicsBodyDefinitions [{
      boundBone {boneName=skirt1_m_r}
      boxExtents=<5 5 5>
      constraintSetup {angularTargetAxis=AxisZ, angularLimitsMax=<0 5 0>, angularLimitsMin=<0 -5 -106.0975>, angularTarget=<0 0 1>}
      localJointOffset=<0.5 0 0>
     }]
    relativeSpaceBone {boneName=pelvis}
   }
   n7 = Anim Dynamics - Bone: skirt1_m_l!(ComponentPose=@n6, Alpha=1, GravityOverride:Vector="0,0,-900", ExternalForce="0,0,0") -> (Pose) {
    angularDampingOverride=1
    angularSpringConstant=950
    bAngularSpring=true
    bOverrideAngularDamping=true
    bUseGravityOverride=true
    simulationSpace=RootRelative
    alphaBoolBlend {blendOption=Linear}
    boundBone {boneName=skirt1_m_l}
    componentAppliedLinearAccClamp=<4000 4000 4000>
    componentLinearAccScale=<0.4 0.4 0.42>
    componentLinearVelScale=<0.4 0.4 0.1>
    gravityOverride {z=-900}
    physicsBodyDefinitions [{
      boundBone {boneName=skirt1_m_l}
      boxExtents=<5 5 5>
      constraintSetup {angularTargetAxis=AxisZ, angularLimitsMax=<0 5 0>, angularLimitsMin=<0 -5 -106.0975>, angularTarget=<0 0 1>}
      localJointOffset=<0.2 0 0>
     }]
    relativeSpaceBone {boneName=pelvis}
   }
   n8 = Anim Dynamics - Bone: rings_cor!(ComponentPose=@n7, Alpha=1) -> (Pose) {
    angularDampingOverride=1
    angularSpringConstant=800
    bAngularSpring=true
    bOverrideAngularDamping=true
    bUseGravityOverride=true
    gravityScale=0
    linearDampingOverride=1
    linearSpringConstant=200
    simulationSpace=RootRelative
    alphaBoolBlend {blendOption=Linear}
    boundBone {boneName=rings_cor}
    componentAppliedLinearAccClamp=<3000 3000 3000>
    componentLinearAccScale=<0.4 0.4 0.4>
    componentLinearVelScale=<0.5 0.5 0.5>
    gravityOverride {z=-980}
    physicsBodyDefinitions [{
      boundBone {boneName=rings_cor}
      boxExtents=<15 15 15>
      constraintSetup {angularLimitsMax=<0 6 5>, angularLimitsMin=<0 -3 -10>, angularTarget=<1 0 0>}
      localJointOffset=<1 0 0>
     }]
    relativeSpaceBone {boneName=lowerarm_l}
   }
   n9 = "Transform (Modify) Bone - Bone: rings_cor"(ComponentPose=@n8, Alpha=1, Translation="0,0,0", Rotation="0,0,0", Scale="1.2,1.2,1.2") {
    rotationMode=BMM_Additive
    rotationSpace=BCS_BoneSpace
    scaleMode=BMM_Additive
    scaleSpace=BCS_BoneSpace
    translationMode=BMM_Additive
    translationSpace=BCS_BoneSpace
    boneToModify {boneName=rings_cor}
    scale=<1.2 1.2 1.2>
   }
   n10 = Anim Dynamics - Bone: rings_cor1!(ComponentPose=@n9, Alpha=1) -> (Pose) {
    angularDampingOverride=1
    angularSpringConstant=800
    bAngularSpring=true
    bOverrideAngularDamping=true
    bUseGravityOverride=true
    gravityScale=0
    linearDampingOverride=1
    linearSpringConstant=200
    simulationSpace=RootRelative
    alphaBoolBlend {blendOption=Linear}
    boundBone {boneName=rings_cor1}
    componentAppliedLinearAccClamp=<3000 3000 3000>
    componentLinearAccScale=<0.4 0.4 0.4>
    componentLinearVelScale=<0.5 0.5 0.5>
    gravityOverride {z=-980}
    physicsBodyDefinitions [{
      boundBone {boneName=rings_cor1}
      boxExtents=<15 15 15>
      constraintSetup {angularLimitsMax=<0 6 10>, angularLimitsMin=<0 -6 -10>, angularTarget=<1 0 0>}
      localJointOffset=<2 0 0>
     }]
    relativeSpaceBone {boneName=lowerarm_l}
   }
   n11 = "Transform (Modify) Bone - Bone: rings_cor1"(ComponentPose=@n10, Alpha=1, Translation="0,0,0", Rotation="0,0,0", Scale="1.15,1.15,1.15") {
    rotationMode=BMM_Additive
    rotationSpace=BCS_BoneSpace
    scaleMode=BMM_Additive
    scaleSpace=BCS_BoneSpace
    translationMode=BMM_Additive
    translationSpace=BCS_BoneSpace
    boneToModify {boneName=rings_cor1}
    scale=<1.15 1.15 1.15>
   }
   n12 = Anim Dynamics - Bone: MSTR_sholder_r(ComponentPose=@n11, Alpha=1, ExternalForce="0,0,0") {
    angularDampingOverride=1
    angularSpringConstant=800
    bAngularSpring=true
    bOverrideAngularDamping=true
    bUseGravityOverride=true
    simulationSpace=RootRelative
    alphaBoolBlend {blendOption=Linear}
    boundBone {boneName=MSTR_sholder_r}
    componentAppliedLinearAccClamp=<4000 4000 4000>
    componentLinearAccScale=<0.4 0.4 0.51>
    componentLinearVelScale=<0.4 0.4 0.1>
    gravityOverride {z=-800}
    physicsBodyDefinitions [{
      boundBone {boneName=MSTR_sholder_r}
      boxExtents=<5 5 5>
      constraintSetup {angularLimitsMax=<0 5 0>, angularLimitsMin=<0 -5 -30>, angularTarget=<1 0 0>}
      localJointOffset=<0.3 0 0>
     }]
    relativeSpaceBone {boneName=clavicle_r}
   }
   n13 = Anim Dynamics - Bone: skirt1_r(ComponentPose=@n12, Alpha=1, ExternalForce="0,0,0") {
    angularDampingOverride=1
    angularSpringConstant=200
    bAngularSpring=true
    bOverrideAngularDamping=true
    bUseGravityOverride=true
    simulationSpace=RootRelative
    alphaBoolBlend {blendOption=Linear}
    boundBone {boneName=skirt1_r}
    componentAppliedLinearAccClamp=<4000 4000 4000>
    componentLinearAccScale=<0.4 0.4 0.4>
    componentLinearVelScale=<0.4 0.4 0.1>
    gravityOverride {z=-900}
    physicsBodyDefinitions [{
      boundBone {boneName=skirt1_r}
      boxExtents=<5 5 5>
      constraintSetup {angularTargetAxis=AxisZ, angularLimitsMax=<0 20 -10>, angularLimitsMin=<0 -20 -115.2438>, angularTarget=<0 0 1>}
      localJointOffset=<0.2 0 0>
     }]
    simSpaceSettings {maxAngularAcceleration=150, maxAngularVelocity=30, simSpaceAngularAlpha=0.4}
   }
   n14 = Anim Dynamics - Bone: skirt2_r(ComponentPose=@n13, Alpha=1, ExternalForce="0,0,0") {
    angularDampingOverride=1
    angularSpringConstant=400
    bAngularSpring=true
    bOverrideAngularDamping=true
    bUseGravityOverride=true
    simulationSpace=RootRelative
    alphaBoolBlend {blendOption=Linear}
    boundBone {boneName=skirt2_r}
    componentAppliedLinearAccClamp=<4000 4000 4000>
    componentLinearAccScale=<0.4 0.4 0.42>
    componentLinearVelScale=<0.4 0.4 0.1>
    gravityOverride {z=-900}
    physicsBodyDefinitions [{
      boundBone {boneName=skirt2_r}
      boxExtents=<5 5 5>
      constraintSetup {angularTargetAxis=AxisZ, angularLimitsMax=<0 20 0>, angularLimitsMin=<0 -20 -106.0975>, angularTarget=<0 0 1>}
      localJointOffset=<0.5 0 0>
     }]
    relativeSpaceBone {boneName=pelvis}
    simSpaceSettings {maxAngularAcceleration=150, maxAngularVelocity=30, simSpaceAngularAlpha=0.4}
   }
   n15 = Anim Dynamics - Bone: skirt3_r(ComponentPose=@n14, Alpha=1, ExternalForce="0,0,0") {
    angularDampingOverride=1
    angularSpringConstant=250
    bAngularSpring=true
    bOverrideAngularDamping=true
    bUseGravityOverride=true
    simulationSpace=RootRelative
    alphaBoolBlend {blendOption=Linear}
    boundBone {boneName=skirt3_r}
    componentAppliedLinearAccClamp=<4000 4000 4000>
    componentLinearAccScale=<0.4 0.4 0.45>
    componentLinearVelScale=<0.4 0.4 0.1>
    gravityOverride {z=-900}
    physicsBodyDefinitions [{
      boundBone {boneName=skirt3_r}
      boxExtents=<5 5 5>
      constraintSetup {angularTargetAxis=AxisZ, angularLimitsMax=<0 20 -10>, angularLimitsMin=<0 -20 -85.9756>, angularTarget=<0 0 1>}
      localJointOffset=<0.2 0 0>
     }]
    simSpaceSettings {maxAngularAcceleration=150, maxAngularVelocity=30, simSpaceAngularAlpha=0.4}
   }
   n16 = Anim Dynamics - Bone: skirt4_r(ComponentPose=@n15, Alpha=1, ExternalForce="0,0,0") {
    angularDampingOverride=1
    angularSpringConstant=200
    bAngularSpring=true
    bOverrideAngularDamping=true
    bUseGravityOverride=true
    simulationSpace=RootRelative
    alphaBoolBlend {blendOption=Linear}
    boundBone {boneName=skirt4_r}
    componentAppliedLinearAccClamp=<4000 4000 4000>
    componentLinearAccScale=<0.4 0.4 0.51>
    componentLinearVelScale=<0.4 0.4 0.1>
    gravityOverride {z=-900}
    physicsBodyDefinitions [{
      boundBone {boneName=skirt4_r}
      boxExtents=<5 5 5>
      constraintSetup {angularTargetAxis=AxisZ, angularLimitsMax=<0 20 0>, angularLimitsMin=<0 -20 -71.9999>, angularTarget=<0 0 1>}
      localJointOffset=<0.1 0 0>
     }]
    relativeSpaceBone {boneName=pelvis}
    simSpaceSettings {maxAngularAcceleration=150, maxAngularVelocity=30, simSpaceAngularAlpha=0.4}
   }
   n17 = Anim Dynamics - Bone: MSTR_sholder_l(ComponentPose=@n16, Alpha=1, ExternalForce="0,0,0") {
    angularDampingOverride=1
    angularSpringConstant=800
    bAngularSpring=true
    bOverrideAngularDamping=true
    bUseGravityOverride=true
    simulationSpace=RootRelative
    alphaBoolBlend {blendOption=Linear}
    boundBone {boneName=MSTR_sholder_l}
    componentAppliedLinearAccClamp=<4000 4000 4000>
    componentLinearAccScale=<0.4 0.4 0.51>
    componentLinearVelScale=<0.4 0.4 0.1>
    gravityOverride {z=-800}
    physicsBodyDefinitions [{
      boundBone {boneName=MSTR_sholder_l}
      boxExtents=<5 5 5>
      constraintSetup {angularLimitsMax=<0 5 0>, angularLimitsMin=<0 -5 -35>, angularTarget=<1 0 0>}
      localJointOffset=<0.1 0 0>
     }]
    relativeSpaceBone {boneName=clavicle_l}
   }
   n18 = Anim Dynamics - Bone: skirt4_l(ComponentPose=@n17, Alpha=1, ExternalForce="0,0,0") {
    angularDampingOverride=1
    angularSpringConstant=250
    bAngularSpring=true
    bOverrideAngularDamping=true
    bUseGravityOverride=true
    simulationSpace=RootRelative
    alphaBoolBlend {blendOption=Linear}
    boundBone {boneName=skirt4_l}
    componentAppliedLinearAccClamp=<4000 4000 4000>
    componentLinearAccScale=<0.4 0.4 0.51>
    componentLinearVelScale=<0.4 0.4 0.1>
    gravityOverride {z=-900}
    physicsBodyDefinitions [{
      boundBone {boneName=skirt4_l}
      boxExtents=<5 5 5>
      constraintSetup {angularTargetAxis=AxisZ, angularLimitsMax=<0 20 -10>, angularLimitsMin=<0 -20 -83.7073>, angularTarget=<0 0 1>}
      localJointOffset=<0.12 0 0>
     }]
   }
   n19 = Anim Dynamics - Bone: skirt3_l(ComponentPose=@n18, Alpha=1, ExternalForce="0,0,0") {
    angularDampingOverride=1
    angularSpringConstant=300
    bAngularSpring=true
    bOverrideAngularDamping=true
    bUseGravityOverride=true
    simulationSpace=RootRelative
    alphaBoolBlend {blendOption=Linear}
    boundBone {boneName=skirt3_l}
    componentAppliedLinearAccClamp=<4000 4000 4000>
    componentLinearAccScale=<0.4 0.4 0.51>
    componentLinearVelScale=<0.4 0.4 0.1>
    gravityOverride {z=-900}
    physicsBodyDefinitions [{
      boundBone {boneName=skirt3_l}
      boxExtents=<5 5 5>
      constraintSetup {angularTargetAxis=AxisZ, angularLimitsMax=<0 20 0>, angularLimitsMin=<0 -20 -93.3788>, angularTarget=<0 0 1>}
      localJointOffset=<0.25 0 0>
     }]
    relativeSpaceBone {boneName=pelvis}
   }
   n20 = Anim Dynamics - Bone: skirt2_l(ComponentPose=@n19, Alpha=1, ExternalForce="0,0,0") {
    angularDampingOverride=1
    angularSpringConstant=200
    bAngularSpring=true
    bOverrideAngularDamping=true
    bUseGravityOverride=true
    simulationSpace=RootRelative
    alphaBoolBlend {blendOption=Linear}
    boundBone {boneName=skirt2_l}
    componentAppliedLinearAccClamp=<4000 4000 4000>
    componentLinearAccScale=<0.4 0.4 0.51>
    componentLinearVelScale=<0.4 0.4 0.1>
    gravityOverride {z=-900}
    physicsBodyDefinitions [{
      boundBone {boneName=skirt2_l}
      boxExtents=<5 5 5>
      constraintSetup {angularTargetAxis=AxisZ, angularLimitsMax=<0 20 -10>, angularLimitsMin=<0 -20 -108.3788>, angularTarget=<0 0 1>}
      localJointOffset=<0.2 0 0>
     }]
   }
   n21 = Anim Dynamics - Bone: skirt1_l(ComponentPose=@n20, Alpha=1, ExternalForce="0,0,0") {
    angularDampingOverride=1
    angularSpringConstant=150
    bAngularSpring=true
    bOverrideAngularDamping=true
    bUseGravityOverride=true
    simulationSpace=RootRelative
    alphaBoolBlend {blendOption=Linear}
    boundBone {boneName=skirt1_l}
    componentAppliedLinearAccClamp=<4000 4000 4000>
    componentLinearAccScale=<0.4 0.4 0.51>
    componentLinearVelScale=<0.4 0.4 0.1>
    gravityOverride {z=-900}
    physicsBodyDefinitions [{
      boundBone {boneName=skirt1_l}
      boxExtents=<5 5 5>
      constraintSetup {angularTargetAxis=AxisZ, angularLimitsMax=<0 20 0>, angularLimitsMin=<0 -20 -113.3932>, angularTarget=<0 0 1>}
      localJointOffset=<0.3 0 0>
     }]
    relativeSpaceBone {boneName=pelvis}
   }
   Output Pose(Result=Component To Local(ComponentPose=@n21)) {layerGroup=None}
   n22 = "Constraint - MSTR_sholder_l (PPPPPPPP)"!(
    ComponentPose=Local To Component(LocalPose=Copy Pose From Mesh() {bUseAttachedParent=true})
    Alpha=1
    ConstraintWeights_0=0.0071
    ConstraintWeights_1:float=0.0602
    ConstraintWeights_2:float=0.4162
    ConstraintWeights_3:float=0.1065
    ConstraintWeights_4:float=0.0481
    ConstraintWeights_5:float=0.0403
    ConstraintWeights_6:float=0.1014
    ConstraintWeights_7:float=0.2202
   ) -> (Pose) {
    boneToModify {boneName=MSTR_sholder_l}
    constraintSetup [
     {targetBone {boneName=spine_04}}
     {targetBone {boneName=spine_05}}
     {targetBone {boneName=clavicle_l}}
     {targetBone {boneName=upperarm_twist_01_l}}
     {targetBone {boneName=deform_upperarm_bk_l}}
     {targetBone {boneName=deform_upperarm_fr_l}}
     {targetBone {boneName=deform_upperarm_up_l}}
     {targetBone {boneName=sholder_l}}
    ]
    constraintWeights [1,1,1,1,1,1,1,1]
   }
   n23 = "Constraint - MSTR_sholder_r (PPPPPPPP)"!(
    ComponentPose=@n22
    Alpha=1
    ConstraintWeights_0=0.0063
    ConstraintWeights_1:float=0.0563
    ConstraintWeights_2:float=0.4098
    ConstraintWeights_3:float=0.1144
    ConstraintWeights_4:float=0.0548
    ConstraintWeights_5:float=0.0377
    ConstraintWeights_6:float=0.1051
    ConstraintWeights_7:float=0.2156
   ) -> (Pose) {
    boneToModify {boneName=MSTR_sholder_r}
    constraintSetup [
     {targetBone {boneName=spine_04}}
     {targetBone {boneName=spine_05}}
     {targetBone {boneName=clavicle_r}}
     {targetBone {boneName=upperarm_twist_01_r}}
     {targetBone {boneName=deform_upperarm_bk_r}}
     {targetBone {boneName=deform_upperarm_fr_r}}
     {targetBone {boneName=deform_upperarm_up_r}}
     {targetBone {boneName=sholder_r}}
    ]
    constraintWeights [1,1,1,1,1,1,1,1]
   }
   n24 = "Constraint - MSTR_skirt_f_r (PPPPPPP)"!(
    ComponentPose=@n23
    Alpha=1
    ConstraintWeights_0=0.0726
    ConstraintWeights_1:float=0.1735
    ConstraintWeights_2:float=0.6162
    ConstraintWeights_3:float=0.1364
    ConstraintWeights_4:float=0
    ConstraintWeights_5:float=0
    ConstraintWeights_6:float=0.0013
   ) -> (Pose) {
    boneToModify {boneName=MSTR_skirt_f_r}
    constraintSetup [
     {targetBone {boneName=pelvis}}
     {targetBone {boneName=spine_01}}
     {targetBone {boneName=spine_02}}
     {targetBone {boneName=spine_03}}
     {targetBone {boneName=deform_groin_l}}
     {targetBone {boneName=thigh_twist_01_r}}
     {targetBone {boneName=deform_groin_r}}
    ]
    constraintWeights [1,1,1,1,1,1,1]
   }
   n25 = "Constraint - MSTR_skirt_d_r (PPPPPP)"!(
    ComponentPose=@n24
    Alpha=1
    ConstraintWeights_0=0.0537
    ConstraintWeights_1:float=0.0223
    ConstraintWeights_2:float=0.1055
    ConstraintWeights_3:float=0.2935
    ConstraintWeights_4:float=0.4844
    ConstraintWeights_5:float=0.0406
   ) -> (Pose) {
    boneToModify {boneName=MSTR_skirt_d_r}
    constraintSetup [
     {targetBone {boneName=spine_02}}
     {targetBone {boneName=skirt2_b_r}}
     {targetBone {boneName=skirt1_d_r}}
     {targetBone {boneName=skirt2_d_r}}
     {targetBone {boneName=skirt1_m_r}}
     {targetBone {boneName=skirt2_f_r}}
    ]
    constraintWeights [1,1,1,1,1,1]
   }
   n26 = "Constraint - MSTR_skirt1_b_r (PPPPP)"!(
    ComponentPose=@n25
    Alpha=1
    ConstraintWeights_0=0.1262
    ConstraintWeights_1:float=0.3005
    ConstraintWeights_2:float=0.5265
    ConstraintWeights_3:float=0.0468
    ConstraintWeights_4:float=0
   ) -> (Pose) {
    boneToModify {boneName=MSTR_skirt1_b_r}
    constraintSetup [
     {targetBone {boneName=pelvis}}
     {targetBone {boneName=spine_01}}
     {targetBone {boneName=spine_02}}
     {targetBone {boneName=spine_03}}
     {targetBone {boneName=deform_groin_r}}
    ]
    constraintWeights [1,1,1,1,1]
   }
   n27 = "Constraint - skirt1_b (PPPPPPPP)"!(
    ComponentPose=@n26
    Alpha=1
    ConstraintWeights_0=0.3616
    ConstraintWeights_1:float=0.2092
    ConstraintWeights_2:float=0.3927
    ConstraintWeights_3:float=0.004
    ConstraintWeights_4:float=0.0036
    ConstraintWeights_5:float=0.0049
    ConstraintWeights_6:float=0.0087
    ConstraintWeights_7:float=0.0153
   ) -> (Pose) {
    boneToModify {boneName=skirt1_b}
    constraintSetup [
     {targetBone {boneName=pelvis}}
     {targetBone {boneName=spine_01}}
     {targetBone {boneName=spine_02}}
     {targetBone {boneName=spine_03}}
     {targetBone {boneName=deform_glute_l}}
     {targetBone {boneName=deform_glute_r}}
     {targetBone {boneName=skirt2_b_l}}
     {targetBone {boneName=skirt2_b_r}}
    ]
    constraintWeights [1,1,1,1,1,1,1,1]
   }
   n28 = "Transform (Modify) Bone - Bone: skirt1_b"(ComponentPose=@n27, Alpha=1, Translation="0,0,0", Rotation="-0.614,7.220,0.163", Scale="1,1,1") {rotationMode=BMM_Additive, rotationSpace=BCS_BoneSpace, boneToModify {boneName=skirt1_b}, rotation=<rot -0.614 7.22 0.163>}
   n29 = "Transform (Modify) Bone - Bone: skirt2_b"(ComponentPose=@n28, Alpha=1, Translation="0,0,0", Rotation="0.622,-10.261,-0.164", Scale="1,1,1") {rotationMode=BMM_Additive, rotationSpace=BCS_BoneSpace, boneToModify {boneName=skirt2_b}, rotation=<rot 0.622 -10.261 -0.164>}
   n30 = "Constraint - MSTR_skirt_b_l (PPPPP)"!(
    ComponentPose=@n29
    Alpha=1
    ConstraintWeights_0=0.1262
    ConstraintWeights_1:float=0.3005
    ConstraintWeights_2:float=0.5265
    ConstraintWeights_3:float=0.0468
    ConstraintWeights_4:float=0
   ) -> (Pose) {
    boneToModify {boneName=MSTR_skirt_b_l}
    constraintSetup [
     {targetBone {boneName=pelvis}}
     {targetBone {boneName=spine_01}}
     {targetBone {boneName=spine_02}}
     {targetBone {boneName=spine_03}}
     {targetBone {boneName=deform_groin_l}}
    ]
    constraintWeights [1,1,1,1,1]
   }
   n31 = "Constraint - MSTR_skirt_d_l (PPPPPP)"!(
    ComponentPose=@n30
    Alpha=1
    ConstraintWeights_0=0.1053
    ConstraintWeights_1:float=0.2653
    ConstraintWeights_2:float=0.5067
    ConstraintWeights_3:float=0.1222
    ConstraintWeights_4:float=0
    ConstraintWeights_5:float=0.0005
   ) -> (Pose) {
    boneToModify {boneName=MSTR_skirt_d_l}
    constraintSetup [
     {targetBone {boneName=pelvis}}
     {targetBone {boneName=spine_01}}
     {targetBone {boneName=spine_02}}
     {targetBone {boneName=spine_03}}
     {targetBone {boneName=thigh_twist_01_l}}
     {targetBone {boneName=deform_groin_l}}
    ]
    constraintWeights [1,1,1,1,1,1]
   }
   n32 = "Constraint - MSTR_skirt_f_l (PPPPPPPP)"!(
    ComponentPose=@n31
    Alpha=1
    ConstraintWeights_0=0.0711
    ConstraintWeights_1:float=0.2101
    ConstraintWeights_2:float=0.5926
    ConstraintWeights_3:float=0.1106
    ConstraintWeights_4:float=0.0067
    ConstraintWeights_5:float=0.0064
    ConstraintWeights_6:float=0.0023
    ConstraintWeights_7:float=0.0002
   ) -> (Pose) {
    boneToModify {boneName=MSTR_skirt_f_l}
    constraintSetup [
     {targetBone {boneName=pelvis}}
     {targetBone {boneName=spine_01}}
     {targetBone {boneName=spine_02}}
     {targetBone {boneName=spine_03}}
     {targetBone {boneName=deform_groin_l}}
     {targetBone {boneName=skirt1_f_l}}
     {targetBone {boneName=skirt2_f_l}}
     {targetBone {boneName=skirt3_f_l}}
    ]
    constraintWeights [1,1,1,1,1,1,1,1]
   }
   n33 = "Transform (Modify) Bone - Bone: MSTR_skirt4_r"(
    ComponentPose="Transform (Modify) Bone - Bone: MSTR_skirt_b"(ComponentPose=@n32, Alpha=1, Translation="0,0,0", Rotation="0,0,3", Scale="1,1,1") {rotationMode=BMM_Additive, rotationSpace=BCS_BoneSpace, boneToModify {boneName=MSTR_skirt_b}, rotation {roll=3}}
    Alpha=1
    Translation="0,0,0"
    Rotation="0,-5,0"
    Scale="1,1,1"
   ) {rotationMode=BMM_Additive, rotationSpace=BCS_BoneSpace, boneToModify {boneName=MSTR_skirt4_r}, rotation {yaw=-5}}
   n34 = Control Rig(Source=Component To Local(ComponentPose="Transform (Modify) Bone - Bone: MSTR_skirt4_l"(ComponentPose=@n33, Alpha=1, Translation="0,0,0", Rotation="0,-5,0", Scale="1,1,1") {rotationMode=BMM_Additive, rotationSpace=BCS_BoneSpace, boneToModify {boneName=MSTR_skirt4_l}, rotation {yaw=-5}})) {controlRigAssetReference {blueprintRigClass=&/F_MED_FishLine_CtrlRig.F_MED_FishLine_CtrlRig_C}}
   Save cached pose 'cech ctrl'(Pose=@n34) {cacheName=cech ctrl}
  ]}
 EventGraph {nodes [
   Event BlueprintUpdateAnimation()
   > n1 = "Gravity Override State Function (Parameter Asset) (Nativized)"(
    CharacterPartAnimInstance=Self-Reference()
    GameTimeInSeconds=GetGameTimeInSeconds()
    Parameters=&/F_MED_FishLine_Parameters
   )
   > Set GravityOverride(GravityOverride=@n1)
   > Set isfullskydivie(isfullskydivie:bool=OR Boolean!(A=Get bIsSkyDiving(), B=Get bIsSkydiveDiveMode(), C:bool=Get bIsSkydiveFloating(), D:bool=Get bIsParachuteOpened()) -> (ReturnValue)) -> (Output_Get:bool)
  ]}
]
variables [
 Character Part Anim Instance {type=CustomCharacterPartAnimInstance, default=null}
 GravityOverride {type=Vector}
 isfullskydivie {type=boolean}
]
