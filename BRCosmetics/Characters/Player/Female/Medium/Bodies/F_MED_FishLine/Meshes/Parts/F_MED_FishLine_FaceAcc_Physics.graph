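# Physics asset for the FishLine female medium character's face accessory part, set up for three chains of back hair (left, middle, right).
# Each chain runs from a small base capsule on the head through three larger capsules, with links limited to wide swing on one axis and very little twist.
# Head, neck, upper spine, clavicle and upper arm capsules supply body shapes around the hair, joined by free-rotation constraints, as are the hair bases.
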PhysicsAsset /BRCosmetics/Characters/Player/Female/Medium/Bodies/F_MED_FishLine/Meshes/Parts/F_MED_FishLine_FaceAcc_Physics
bodies {
 base_backHair_L_01 [{shapeType=Capsule, length=0, radius=1.5, shapeName=base_backHair_L_01_capsule}]
 base_backHair_M_01 [{shapeType=Capsule, length=0, radius=1.5, shapeName=base_backHair_M_01_capsule}]
 base_backHair_R_01 [{shapeType=Capsule, length=0, radius=1.5, shapeName=base_backHair_R_01_capsule}]
 clavicle_l [{shapeType=Capsule, length=3, radius=6.5, shapeName=clavicle_l_capsule, center=<4.5 -0.1197 0.5>, rotation=<rot 0 90 -90>}]
 clavicle_r [{shapeType=Capsule, length=7.3729, radius=6.5312, shapeName=clavicle_r_capsule, center=<-4.5 0.7314 0.2664>, rotation=<rot 0 90 89.8205>}]
 dyn_backHair_L_01 [{
   shapeType=Capsule
   length=5.9296
   radius=3.6277
   shapeName=dyn_backHair_L_01_capsule
   center=<4.3 -0.5262 -4.7759>
   rotation=<rot 0 0 -7.2978>
  }]
 dyn_backHair_L_02 [{shapeType=Capsule, length=6, radius=5.2096, shapeName=dyn_backHair_L_02_capsule, center=<2.6 -5 0>}]
 dyn_backHair_L_03 [{shapeType=Capsule, length=7, radius=6.1349, shapeName=dyn_backHair_L_03_capsule, center=<5.4 -5.7 0>, rotation=<rot 0 0 -0.1795>}]
 dyn_backHair_M_01 [{shapeType=Capsule, length=6, radius=4, shapeName=dyn_backHair_M_01_capsule, center=<3.2 0 0>}]
 dyn_backHair_M_02 [{shapeType=Capsule, length=6, radius=4.3606, shapeName=dyn_backHair_M_02_capsule, center=<2.6 -3 0>, rotation=<rot 0 0 0.1874>}]
 dyn_backHair_M_03 [{shapeType=Capsule, length=7, radius=5.8516, shapeName=dyn_backHair_M_03_capsule, center=<5.4 -5.7 0>, rotation=<rot 0 0 0.1874>}]
 dyn_backHair_R_01 [{shapeType=Capsule, length=5.5, radius=3.9746, shapeName=dyn_backHair_R_01_capsule, center=<4.3 -0.8726 4.0459>}]
 dyn_backHair_R_02 [{shapeType=Capsule, length=6, radius=5.4554, shapeName=dyn_backHair_R_02_capsule, center=<2.6 -5 0>, rotation=<rot 0 0 -0.1942>}]
 dyn_backHair_R_03 [{shapeType=Capsule, length=7, radius=6.1441, shapeName=dyn_backHair_R_03_capsule, center=<5.4 -5.7 0>}]
 head [{shapeType=Capsule, length=10, radius=7.4, shapeName=head_capsule, center=<-3.1505 3.6634 0.007>, rotation=<rot -90 -0.3595 0>}]
 neck_01 [{shapeType=Capsule, length=9.5, radius=3, shapeName=neck_01_capsule, center=<-0.5841 1.1832 0>, rotation=<rot 0 -90 0>}]
 spine_04 [{shapeType=Capsule, length=5, radius=9, shapeName=spine_04_capsule, center=<1.9213 0.7007 0>, rotation=<rot 0 -90 0>}]
 spine_05 [
  {shapeType=Capsule, length=12, radius=7.5, shapeName=spine_05_capsule, center=<3.0474 0.9099 0>, rotation=<rot 0 -90 0>}
  {shapeType=Capsule, length=12.5, radius=10.4, shapeName=spine_05_capsule, center=<-3.9521 2.7659 0>}
 ]
 upperarm_l [{shapeType=Capsule, length=13.5, radius=4.5, shapeName=upperarm_l_capsule, center=<6.0094 0 -1.1471>, rotation=<rot -85 179.9999 179.8444>}]
 upperarm_r [{shapeType=Capsule, length=13.5, radius=4.5, shapeName=upperarm_r_capsule, center=<-6.0511 0 1.1507>, rotation=<rot -85 180 -180>}]
}
constraints [
 {
  bone1Name=head
  bone2Name=neck_01
  swing1LimitDegrees=45
  swing1Motion=Free
  swing2LimitDegrees=45
  swing2Motion=Free
  twistLimitDegrees=45
  twistMotion=Free
 }
 {
  bone1Name=clavicle_l
  bone2Name=spine_05
  swing1LimitDegrees=45
  swing1Motion=Free
  swing2LimitDegrees=45
  swing2Motion=Free
  twistLimitDegrees=45
  twistMotion=Free
 }
 {
  bone1Name=clavicle_r
  bone2Name=spine_05
  swing1LimitDegrees=45
  swing1Motion=Free
  swing2LimitDegrees=45
  swing2Motion=Free
  twistLimitDegrees=45
  twistMotion=Free
 }
 {
  bone1Name=neck_01
  bone2Name=spine_05
  swing1LimitDegrees=45
  swing1Motion=Free
  swing2LimitDegrees=45
  swing2Motion=Free
  twistLimitDegrees=45
  twistMotion=Free
 }
 {
  bone1Name=upperarm_l
  bone2Name=clavicle_l
  swing1LimitDegrees=45
  swing1Motion=Free
  swing2LimitDegrees=45
  swing2Motion=Free
  twistLimitDegrees=45
  twistMotion=Free
 }
 {
  bone1Name=upperarm_r
  bone2Name=clavicle_r
  swing1LimitDegrees=45
  swing1Motion=Free
  swing2LimitDegrees=45
  swing2Motion=Free
  twistLimitDegrees=45
  twistMotion=Free
 }
 {
  bone1Name=spine_05
  bone2Name=spine_04
  swing1LimitDegrees=45
  swing1Motion=Free
  swing2LimitDegrees=45
  swing2Motion=Free
  twistLimitDegrees=45
  twistMotion=Free
 }
 {
  bone1Name=base_backHair_L_01
  bone2Name=head
  swing1LimitDegrees=20
  swing1Motion=Free
  swing2LimitDegrees=10
  swing2Motion=Free
  twistLimitDegrees=1
  twistMotion=Free
 }
 {bone1Name=dyn_backHair_L_01, bone2Name=base_backHair_L_01, swing1LimitDegrees=60, swing2LimitDegrees=10, twistLimitDegrees=1}
 {bone1Name=dyn_backHair_L_02, bone2Name=dyn_backHair_L_01, swing1LimitDegrees=60, swing2LimitDegrees=10, twistLimitDegrees=1}
 {bone1Name=dyn_backHair_L_03, bone2Name=dyn_backHair_L_02, swing1LimitDegrees=60, swing2LimitDegrees=10, twistLimitDegrees=1}
 {
  bone1Name=base_backHair_M_01
  bone2Name=head
  swing1LimitDegrees=20
  swing1Motion=Free
  swing2LimitDegrees=10
  swing2Motion=Free
  twistLimitDegrees=1
  twistMotion=Free
 }
 {bone1Name=dyn_backHair_M_01, bone2Name=base_backHair_M_01, swing1LimitDegrees=60, swing2LimitDegrees=10, twistLimitDegrees=1}
 {bone1Name=dyn_backHair_M_02, bone2Name=dyn_backHair_M_01, swing1LimitDegrees=60, swing2LimitDegrees=10, twistLimitDegrees=1}
 {bone1Name=dyn_backHair_M_03, bone2Name=dyn_backHair_M_02, swing1LimitDegrees=60, swing2LimitDegrees=10, twistLimitDegrees=1}
 {
  bone1Name=base_backHair_R_01
  bone2Name=head
  swing1LimitDegrees=20
  swing1Motion=Free
  swing2LimitDegrees=10
  swing2Motion=Free
  twistLimitDegrees=1
  twistMotion=Free
 }
 {bone1Name=dyn_backHair_R_01, bone2Name=base_backHair_R_01, swing1LimitDegrees=60, swing2LimitDegrees=10, twistLimitDegrees=1}
 {bone1Name=dyn_backHair_R_02, bone2Name=dyn_backHair_R_01, swing1LimitDegrees=60, swing2LimitDegrees=10, twistLimitDegrees=1}
 {bone1Name=dyn_backHair_R_03, bone2Name=dyn_backHair_R_02, swing1LimitDegrees=60, swing2LimitDegrees=10, twistLimitDegrees=1}
]
