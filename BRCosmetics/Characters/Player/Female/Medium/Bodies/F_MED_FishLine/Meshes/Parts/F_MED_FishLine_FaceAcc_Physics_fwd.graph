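# Physics asset for the face accessory part of the female medium FishLine body, simulating two front hair strands, left and right.
# Each strand is a four-capsule dynamic chain with tight twist and swing limits, hung from a base bone locked to neck_01.
# Capsules on spine_05, neck, clavicles and upper arms, joined by free-angle constraints, appear to serve as body colliders for the strands.
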PhysicsAsset /BRCosmetics/Characters/Player/Female/Medium/Bodies/F_MED_FishLine/Meshes/Parts/F_MED_FishLine_FaceAcc_Physics_fwd
bodies {
 base_hair_Front_L_01 [{
   shapeType=Capsule
   length=6.8686
   radius=2.2004
   shapeName=base_hair_Front_L_01_capsule
   center=<3.146 -0.5136 8.0695>
   rotation=<rot -5.7884 -90 5.8679>
  }]
 base_hair_Front_R_01 [{
   shapeType=Capsule
   length=2.9808
   radius=2.2183
   shapeName=base_hair_Front_R_01_capsule
   center=<4.1418 0.2746 0.1561>
   rotation=<rot 0 143.8688 -62.6662>
  }]
 clavicle_l [{shapeType=Capsule, length=8, radius=5, shapeName=clavicle_l_capsule, center=<3.7261 -1 0.0443>, rotation=<rot 0 90 -90>}]
 clavicle_r [{shapeType=Capsule, length=8, radius=5, shapeName=clavicle_r_capsule, center=<-3.7261 1 -0.0443>, rotation=<rot 0 90 90>}]
 dyn_hair_Front_L_01 [{shapeType=Capsule, length=3, radius=3, shapeName=dyn_hair_Front_L_01_capsule, center=<2.4448 0 0>, rotation=<rot -0.0008 -0.0095 -0.2>}]
 dyn_hair_Front_L_02 [{shapeType=Capsule, length=2.5, radius=2.5, shapeName=dyn_hair_Front_L_02_capsule, center=<2.4448 0 0>, rotation=<rot 89.4435 179.9999 180>}]
 dyn_hair_Front_L_03 [{
   shapeType=Capsule
   length=3.5
   radius=2.5
   shapeName=dyn_hair_Front_L_03_capsule
   center=<2.4448 0 0>
   rotation=<rot -89.5163 179.9907 -179.9954>
  }]
 dyn_hair_Front_L_04 [{shapeType=Capsule, length=3.5, radius=2.5, shapeName=dyn_hair_Front_L_04_capsule, center=<2.4448 0 0>, rotation=<rot 86.4578 -0.0001 0>}]
 dyn_hair_Front_R_01 [{shapeType=Capsule, length=3, radius=3, shapeName=dyn_hair_Front_R_01_capsule, center=<2.4448 0 0>, rotation=<rot 0 0 -0.1998>}]
 dyn_hair_Front_R_02 [{
   shapeType=Capsule
   length=2.5
   radius=2.5
   shapeName=dyn_hair_Front_R_02_capsule
   center=<2.4448 0 0>
   rotation=<rot 89.4435 179.9995 179.9998>
  }]
 dyn_hair_Front_R_03 [{
   shapeType=Capsule
   length=3.5
   radius=2.5
   shapeName=dyn_hair_Front_R_03_capsule
   center=<2.4448 0 0>
   rotation=<rot -89.5164 -179.9629 179.9582>
  }]
 dyn_hair_Front_R_04 [{shapeType=Capsule, length=3.5, radius=2.5, shapeName=dyn_hair_Front_R_04_capsule, center=<3.9464 0.2629 -0.093>, rotation=<rot 86.4578 0 0>}]
 neck_01 [
  {shapeType=Capsule, length=8.005, radius=4.5, shapeName=neck_01_capsule, center=<2.6041 0 0.0013>, rotation=<rot 0 -90.0006 89.9721>}
  {shapeType=Capsule, length=25.1131, radius=4.8869, shapeName=neck_01_capsule, center=<0.5915 -5.5298 0>}
 ]
 spine_05 [
  {shapeType=Capsule, length=33, radius=6.9, shapeName=spine_05_capsule, center=<-4.5905 3.9759 0>}
  {shapeType=Capsule, length=15, radius=6.5, shapeName=spine_05_capsule}
 ]
 upperarm_l [{shapeType=Capsule, length=25, radius=5, shapeName=upperarm_l_capsule, center=<13.6001 -2.0175 1.1076>, rotation=<rot -85 180 -180>}]
 upperarm_r [{shapeType=Capsule, length=25, radius=5, shapeName=upperarm_r_capsule, center=<-14.3134 1.6218 -1.1559>, rotation=<rot -85 180 -180>}]
}
constraints [
 {
  bone1Name=clavicle_l
  bone2Name=spine_05
  swing1LimitDegrees=45
  swing1Motion=Free
  swing2LimitDegrees=45
  swing2Motion=Free
  twistLimitDegrees=45
  twistMotion=Free
 }
 {
  bone1Name=clavicle_r
  bone2Name=spine_05
  swing1LimitDegrees=45
  swing1Motion=Free
  swing2LimitDegrees=45
  swing2Motion=Free
  twistLimitDegrees=45
  twistMotion=Free
 }
 {
  bone1Name=upperarm_l
  bone2Name=clavicle_l
  swing1LimitDegrees=45
  swing1Motion=Free
  swing2LimitDegrees=45
  swing2Motion=Free
  twistLimitDegrees=45
  twistMotion=Free
 }
 {
  bone1Name=upperarm_r
  bone2Name=clavicle_r
  swing1LimitDegrees=45
  swing1Motion=Free
  swing2LimitDegrees=45
  swing2Motion=Free
  twistLimitDegrees=45
  twistMotion=Free
 }
 {
  bone1Name=base_hair_Front_L_01
  bone2Name=neck_01
  swing1LimitDegrees=45
  swing1Motion=Locked
  swing2LimitDegrees=45
  swing2Motion=Locked
  twistLimitDegrees=1
  twistMotion=Locked
 }
 {bone1Name=dyn_hair_Front_L_01, bone2Name=base_hair_Front_L_01, swing1LimitDegrees=20, swing2LimitDegrees=1, twistLimitDegrees=1}
 {bone1Name=dyn_hair_Front_L_02, bone2Name=dyn_hair_Front_L_01, swing1LimitDegrees=20, swing2LimitDegrees=10, twistLimitDegrees=1}
 {bone1Name=dyn_hair_Front_L_03, bone2Name=dyn_hair_Front_L_02, swing1LimitDegrees=25, swing2LimitDegrees=10, twistLimitDegrees=1}
 {bone1Name=dyn_hair_Front_L_04, bone2Name=dyn_hair_Front_L_03, swing1LimitDegrees=25, swing2LimitDegrees=10, twistLimitDegrees=1}
 {
  bone1Name=base_hair_Front_R_01
  bone2Name=neck_01
  swing1LimitDegrees=45
  swing1Motion=Locked
  swing2LimitDegrees=45
  swing2Motion=Locked
  twistLimitDegrees=1
  twistMotion=Locked
 }
 {bone1Name=dyn_hair_Front_R_01, bone2Name=base_hair_Front_R_01, swing1LimitDegrees=20, swing2LimitDegrees=1, twistLimitDegrees=1}
 {bone1Name=dyn_hair_Front_R_02, bone2Name=dyn_hair_Front_R_01, swing1LimitDegrees=20, swing2LimitDegrees=10, twistLimitDegrees=1}
 {bone1Name=dyn_hair_Front_R_03, bone2Name=dyn_hair_Front_R_02, swing1LimitDegrees=25, swing2LimitDegrees=10, twistLimitDegrees=1}
 {bone1Name=dyn_hair_Front_R_04, bone2Name=dyn_hair_Front_R_03, swing1LimitDegrees=25, swing2LimitDegrees=10, twistLimitDegrees=1}
 {
  bone1Name=neck_01
  bone2Name=spine_05
  swing1LimitDegrees=45
  swing1Motion=Free
  swing2LimitDegrees=45
  swing2Motion=Free
  twistLimitDegrees=45
  twistMotion=Free
 }
]
