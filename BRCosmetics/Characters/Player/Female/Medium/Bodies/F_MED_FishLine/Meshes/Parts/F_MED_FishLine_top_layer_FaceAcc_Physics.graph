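# Physics asset for the top-layer FaceAcc mesh part of the female medium FishLine player cosmetic, providing bodies for its hair bones.
# A three-capsule top-hair chain hangs off the base_backHair_M_01 sphere with narrow swing and very tight twist limits.
# A two-capsule middle back-hair chain shares that anchor, while a left back-hair capsule joins the hairMaster sphere.
# All back-hair joints leave swing and twist free.
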PhysicsAsset /BRCosmetics/Characters/Player/Female/Medium/Bodies/F_MED_FishLine/Meshes/Parts/F_MED_FishLine_top_layer_FaceAcc_Physics
bodies {
 Dyn_hair_top_M_01 [{shapeType=Capsule, length=10, radius=2.5, shapeName=Dyn_hair_top_M_01_capsule, center=<4.1771 0 -0.7389>, rotation=<rot 90 90 0>}]
 Dyn_hair_top_M_02 [{shapeType=Capsule, length=10, radius=2.5, shapeName=Dyn_hair_top_M_02_capsule, center=<2.1816 0 -0.7389>, rotation=<rot 90 90 0>}]
 Dyn_hair_top_M_03 [{shapeType=Capsule, length=9.5, radius=3, shapeName=Dyn_hair_top_M_03_capsule, center=<3.461 0 0>, rotation=<rot 90 90 0>}]
 base_backHair_M_01 [{shapeType=Sphere, radius=0.505, shapeName=base_backHair_M_01_sphere}]
 dyn_backHair_L_01 [{shapeType=Capsule, length=12.8777, radius=5.4469, shapeName=dyn_backHair_L_01_capsule, center=<13.1048 5.4437 -7.934>}]
 dyn_backHair_M_01 [{shapeType=Capsule, length=15, radius=15, shapeName=dyn_backHair_M_01_capsule, center=<5.4052 16.2036 0>, rotation=<rot 0 0 0.2735>}]
 dyn_backHair_M_02 [{shapeType=Capsule, length=15, radius=15, shapeName=dyn_backHair_M_02_capsule, center=<10.0025 11.4745 0>, rotation=<rot 0 -14.942 0>}]
 hairMaster [{shapeType=Sphere, radius=0.505, shapeName=hairMaster_sphere}]
}
constraints [
 {bone1Name=Dyn_hair_top_M_02, bone2Name=Dyn_hair_top_M_01, swing1LimitDegrees=10, swing2LimitDegrees=30, twistLimitDegrees=1}
 {bone1Name=Dyn_hair_top_M_03, bone2Name=Dyn_hair_top_M_02, swing1LimitDegrees=10, swing2LimitDegrees=30, twistLimitDegrees=1}
 {bone1Name=Dyn_hair_top_M_01, bone2Name=base_backHair_M_01, swing1LimitDegrees=5, swing2LimitDegrees=20, twistLimitDegrees=1}
 {
  bone1Name=dyn_backHair_M_01
  bone2Name=base_backHair_M_01
  swing1LimitDegrees=45
  swing1Motion=Free
  swing2LimitDegrees=45
  swing2Motion=Free
  twistLimitDegrees=45
  twistMotion=Free
 }
 {
  bone1Name=dyn_backHair_M_02
  bone2Name=dyn_backHair_M_01
  swing1LimitDegrees=45
  swing1Motion=Free
  swing2LimitDegrees=45
  swing2Motion=Free
  twistLimitDegrees=45
  twistMotion=Free
 }
 {
  bone1Name=dyn_backHair_L_01
  bone2Name=hairMaster
  swing1LimitDegrees=45
  swing1Motion=Free
  swing2LimitDegrees=45
  swing2Motion=Free
  twistLimitDegrees=45
  twistMotion=Free
 }
]
